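# Niagara system for a blue firework effect in the ScoopPluck player emote: a rising flare that bursts offset in front of the origin.
# A short-lived parent flare launches upward with impact streaks, while ribbon smoke and trail embers spawn from it via particle reads.
# The burst layers background glow, dust, mesh explosion, flash with light, and twin CPU/GPU spark bursts.
# Brightness of several layers follows a time-of-day multiplier; two flare-explosion parent emitters are disabled.
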NiagaraSystem /BRCosmetics/Animation/Game/MainPlayer/Emotes/ScoopPluck/FX/NSScoopPluck_FireworkBlue1
dependencies {
 usedDataInterfaces [
  &NiagaraDataInterfaceCurve
  &NiagaraDataInterfaceParticleRead
  &NiagaraDataInterfaceArrayFloat
  &NiagaraDataInterfaceColorCurve
  &NiagaraDataInterfaceVector2DCurve
 ]
 usedDynamicInputs [
  &MakeVector4_FromFloat
  &MakeVector2D
  &FNTimeOfDayScalar
 ]
 usedModules [
  &SystemState
  &self:SystemSpawnScript.SystemScriptSource.SystemScriptGraph.NiagaraNodeAssignment_2.SetVariables_#1
 ]
}
emitters {
 BG_Glow002 {
  emitterUpdateScript [
   &EmitterState()
   &SpawnBurst_Instantaneous(Spawn Count=1, Use Spawn Probability=false)
  ]
  particleSpawnScript [
   &InitializeParticle(
    Color=<rgba 0.1769 0.0064 0.3073 0.5>
    Lifetime=0.5
    Position Offset=<75 0 125>
    Position Offset Coordinate Space=::Local
    Sprite Size Mode=::NewEnumerator3
    Uniform Sprite Size=120
    UsePositionOffset=true
    Write Material Random=true
   )
   &CameraOffset(Camera Offset Amount=-20)
  ]
  particleUpdateScript [
   &ParticleState()
   &ScaleSpriteSize(Scale Factor=&Vector2DFromFloat(Value=&FloatFromCurve(FloatCurve=DataInterface() {propertyValues {Curve {keys [
          {arriveTangent=0.4346, interpMode=RCIM_Cubic, leaveTangent=0.4346, tangentMode=RCTM_User, time=0, value=0}
          {arriveTangent=0.0473, interpMode=RCIM_Cubic, leaveTangent=0.0473, tangentMode=RCTM_User, time=0.2, value=1}
         ]}}})))
   &ScaleColor(
    Scale Alpha=&FloatFromCurve(
     FloatCurve=DataInterface() {propertyValues {Curve {keys [
         {arriveTangent=0.4005, interpMode=RCIM_Cubic, leaveTangent=0.4005, tangentMode=RCTM_User, time=0, value=0}
         {arriveTangent=0.0256, interpMode=RCIM_Cubic, leaveTangent=0.0256, tangentMode=RCTM_User, time=0.2, value=1}
         <at 1 0 RCIM_Cubic>
        ]}}}
     Scale Curve=$System.TimeOfDayBrightnessMultiplier
    )
    ScaleA=true
    ScaleRGB=true
   )
   /SetVariables(
    Particles.GlowColor2:LinearColor=&MakeLinearColorFromVectorAndFloat(
     "Vector (RGB)"=&MakeVectorFromLinearColorRGB(Color=<rgba 0.4427 0.6706 5 1>)
     "Float (Alpha)"=&Multiply_Float(
      A=0.2
      B=&FloatFromCurve(FloatCurve=DataInterface() {propertyValues {Curve {keys [
           {arriveTangent=0.4005, interpMode=RCIM_Cubic, leaveTangent=0.4005, tangentMode=RCTM_User, time=0, value=0}
           {arriveTangent=0.0256, interpMode=RCIM_Cubic, leaveTangent=0.0256, tangentMode=RCTM_User, time=0.2, value=1}
           <at 1 0 RCIM_Cubic>
          ]}}})
     )
    )
    Particles.GlowSize2:Vector2f=&Vector2DFromFloat(Value=&Multiply_Float(
      A=120
      B=&FloatFromCurve(FloatCurve=DataInterface() {propertyValues {Curve {keys [
           {arriveTangent=0.4346, interpMode=RCIM_Cubic, leaveTangent=0.4346, tangentMode=RCTM_User, time=0, value=0}
           {arriveTangent=0.0473, interpMode=RCIM_Cubic, leaveTangent=0.0473, tangentMode=RCTM_User, time=0.2, value=1}
          ]}}})
     ))
   ) {isSetParameters=true}
  ]
  rendererClasses [&NiagaraSpriteRendererProperties]
  renderers [
   {rendererClass=&NiagaraSpriteRendererProperties}
   {rendererClass=&NiagaraSpriteRendererProperties, rendererIndex=1}
  ]
 }
 Dust002 {
  emitterUpdateScript [
   &EmitterState(
    Enable Distance Culling=true
    MaxDistance=10000
    Module.bUseMaxDistance=true
    Module.bUseMinDistance=false
    Scalability Mode=::NewEnumerator1
   )
   &SpawnBurst_Instantaneous(Spawn Count=8, Use Spawn Probability=false, UseLoopCountLimit=false)
  ]
  particleSpawnScript [
   &InitializeParticle(
    Lifetime Max=2.6
    Lifetime Min=0.7
    Lifetime Mode=::NewEnumerator1
    Position Offset=<75 0 125>
    Position Offset Coordinate Space=::Local
    Sprite Rotation Mode=::NewEnumerator2
    Sprite Size Mode=::NewEnumerator1
    Uniform Sprite Size Max=35
    Uniform Sprite Size Min=15
    UsePositionOffset=true
    Write Material Random=true
   )
   &ShapeLocation(MaskSpawnGroup=false, Offset Mode=::NewEnumerator0, Override Seed=false, Sphere Radius=5, UseOwnerScale=true)
   &AddVelocity(
    Module.Use Falloff Distance=false
    Origin Offset=<75 0 150>
    Override Seed=false
    Velocity Mode=::NewEnumerator1
    Velocity Speed=&RandomRangeFloat(Minimum=333, Maximum=555)
   )
   &CameraOffset(Camera Offset Amount=&RandomRangeFloat(Minimum=-10, Maximum=-50, Evaluation Type=::NewEnumerator1))
  ]
  particleUpdateScript [
   &ParticleState()
   &ScaleSpriteSize(
    Scale Sprite Size Mode=::NewEnumerator0
    Uniform Scale Factor=&FloatFromCurve(
     FloatCurve=DataInterface() {propertyValues {Curve {keys [
         {interpMode=RCIM_Cubic, tangentMode=RCTM_User, time=0, value=0.3}
         {arriveTangent=0.5158, interpMode=RCIM_Cubic, leaveTangent=0.5158, tangentMode=RCTM_User, time=0.4, value=0.9}
         {arriveTangentWeight=0.3824, interpMode=RCIM_Cubic, tangentMode=RCTM_User, time=1, value=1}
        ]}}}
     Scale Curve=1.5
    )
   )
   &Drag(Drag=16, Use Linear Drag=true, Use Rotational Drag=true)
   &GravityForce(Gravity=<0 0 -222>)
   &SolveForcesAndVelocity(Clamp Velocity=false, Limit Acceleration=false)
   &SpriteRotationRate(Rotation Rate=&RandomRangeFloat(Minimum=-0.03, Maximum=0.03))
   &ScaleColor(
    Scale Alpha=&FloatFromCurve(
     FloatCurve=DataInterface() {propertyValues {Curve {keys [
         {arriveTangent=9.0725, interpMode=RCIM_Cubic, leaveTangent=9.0725, tangentMode=RCTM_User, time=0, value=0}
         {arriveTangent=-2.5249, interpMode=RCIM_Cubic, leaveTangent=-2.5249, tangentMode=RCTM_User, time=0.1408, value=0.9847}
         {arriveTangent=-0.0405, interpMode=RCIM_Cubic, leaveTangent=-0.0405, tangentMode=RCTM_User, time=1, value=0}
        ]}}}
     Scale Curve=0.05
    )
    Scale RGB=&VectorFromFloat(Value=$System.TimeOfDayBrightnessMultiplier)
    ScaleA=true
    ScaleRGB=true
   )
   &DynamicMaterialParameters(
    Index 0 Param 1=&Add_Float(
     A=&RandomRangeFloat(Minimum=0, Maximum=1)
     B=&FloatFromCurve(
      FloatCurve=DataInterface() {propertyValues {Curve {keys [
          {arriveTangent=2.7533, interpMode=RCIM_Cubic, leaveTangent=1.8958, tangentMode=RCTM_User, time=0, value=1}
          {arriveTangent=0.8292, interpMode=RCIM_Cubic, leaveTangent=0.8292, tangentMode=RCTM_User, time=0.321, value=1.3453}
          {arriveTangent=0.0529, interpMode=RCIM_Cubic, leaveTangent=0.2021, tangentMode=RCTM_User, time=1, value=1.5}
         ]}}}
      CurveIndex=&Add_Float(A=&Multiply_Float(A=$Particles.Initial.MaterialRandom, B=0.5), B=$Particles.NormalizedAge)
      Scale Curve=&Lerp_Float(A=1, B=2, Alpha=0.5)
     )
    )
    Index 0 Param 2=&FloatFromCurve(FloatCurve=DataInterface() {propertyValues {Curve {keys [<at 0 1>,<at 1 0>]}}})
    Index 0 Param 3=&FloatFromCurve(
     FloatCurve=DataInterface() {propertyValues {Curve {keys [
         {arriveTangent=-2.2052, interpMode=RCIM_Cubic, leaveTangent=-2.2052, tangentMode=RCTM_User, time=0.1, value=1}
         {interpMode=RCIM_Cubic, tangentMode=RCTM_User, time=1, value=0.4}
        ]}}}
     Scale Curve=&Multiply_Float(A=0.5, B=$System.TimeOfDayBrightnessMultiplier)
    )
    Index 0 Param 4=&FloatFromCurve(FloatCurve=DataInterface() {propertyValues {Curve {keys [{interpMode=RCIM_Cubic, tangentMode=RCTM_Break, time=0.5, value=1}, {interpMode=RCIM_Cubic, tangentMode=RCTM_Break, time=1, value=0}]}}})
    Index 1 Param 1=&FloatFromCurve(
     FloatCurve=DataInterface() {propertyValues {Curve {keys [
         {interpMode=RCIM_Cubic, leaveTangentWeight=0.0833, tangentMode=RCTM_Break, time=0, value=0}
         {arriveTangentWeight=0.0873, interpMode=RCIM_Cubic, tangentMode=RCTM_Break, time=0.25, value=1}
        ]}}}
     Scale Curve=&Lerp_Float(A=0.35, B=0.6, Alpha=0.5)
    )
    Index 1 Param 2=&FloatFromCurve(
     FloatCurve=DataInterface() {propertyValues {Curve {keys [
         {interpMode=RCIM_Cubic, tangentMode=RCTM_Break, time=0, value=0.2}
         {interpMode=RCIM_Cubic, leaveTangent=0.9066, tangentMode=RCTM_Break, time=1, value=2}
        ]}}}
     Scale Curve=&RandomRangeFloat(Minimum=0.1, Maximum=0.2)
    )
    Index 1 Param 3=&FloatFromCurve(FloatCurve=DataInterface() {propertyValues {Curve {keys [
         {interpMode=RCIM_Cubic, leaveTangent=6.3917, tangentMode=RCTM_Break, time=0, value=0}
         {arriveTangent=1.7549, interpMode=RCIM_Cubic, leaveTangent=10.6767, tangentMode=RCTM_User, time=0.25, value=1}
        ]}}})
    Index 1 Param 4=&FloatFromCurve(
     FloatCurve=DataInterface() {propertyValues {Curve {keys [{arriveTangent=-0.136, interpMode=RCIM_Cubic, leaveTangent=-0.136, tangentMode=RCTM_User, time=0, value=1}, <at 1 0.95>]}}}
     Scale Curve=&Lerp_Float(A=2, B=5, Alpha=0.5)
    )
    Index 2 Param 1=&FloatFromCurve(FloatCurve=DataInterface() {propertyValues {Curve {keys [
         {interpMode=RCIM_Cubic, leaveTangent=4.2485, tangentMode=RCTM_Break, time=0, value=0}
         {arriveTangent=1.0548, interpMode=RCIM_Cubic, leaveTangent=1.0548, tangentMode=RCTM_User, time=0.17, value=0.5715}
         <at 1 1>
        ]}}})
    Index 2 Param 2=&FloatFromCurve(
     FloatCurve=DataInterface() {propertyValues {Curve {keys [
         {arriveTangent=1.9056, interpMode=RCIM_Cubic, leaveTangent=1.9056, tangentMode=RCTM_User, time=0, value=0.2}
         {arriveTangent=0.0259, interpMode=RCIM_Cubic, leaveTangent=0.0259, tangentMode=RCTM_User, time=1, value=1.4}
        ]}}}
     Scale Curve=1.5
    )
    Index 2 Param 3=&FloatFromCurve(
     FloatCurve=DataInterface() {propertyValues {Curve {keys [
         {interpMode=RCIM_Cubic, leaveTangentWeight=0.1411, tangentMode=RCTM_Break, time=0, value=2}
         {interpMode=RCIM_Cubic, tangentMode=RCTM_Break, time=0.3, value=1}
         {interpMode=RCIM_Cubic, leaveTangent=11.1233, tangentMode=RCTM_Break, time=1, value=2}
        ]}}}
     Scale Curve=10
    )
    Index 2 Param 4=&FloatFromCurve(
     FloatCurve=DataInterface() {propertyValues {Curve {keys [
         {arriveTangent=-0.6867, interpMode=RCIM_Cubic, leaveTangent=-0.6867, tangentMode=RCTM_User, time=0, value=1}
         {interpMode=RCIM_Cubic, leaveTangent=-1.2956, tangentMode=RCTM_Break, time=1, value=0.75}
        ]}}}
     Scale Curve=&RandomRangeFloat(Minimum=0.5, Maximum=2)
    )
    Module.Param0WriteEnabled=true
    Module.Param0WriteEnabled1=true
    Module.Param0WriteEnabled2=true
    Module.Param1WriteEnabled=false
    Module.Param1WriteEnabled1=true
    Module.Param1WriteEnabled2=true
    Module.Param2WriteEnabled=true
    Module.Param2WriteEnabled1=true
    Module.Param2WriteEnabled2=true
    Module.Param3WriteEnabled=true
    Module.Param3WriteEnabled1=true
    Module.Param3WriteEnabled2=true
    Write Parameter Index 1=true
    Write Parameter Index 2=true
   )
  ]
  rendererClasses [&NiagaraSpriteRendererProperties]
  renderers [
   {rendererClass=&NiagaraSpriteRendererProperties}
   {rendererClass=&NiagaraSpriteRendererProperties, rendererIndex=1}
  ]
 }
 FireworkFlare_Explosion_Final_Parent {
  bEnabled=false
  simTarget=GPUComputeSim
  emitterUpdateScript [
   &EmitterState()
   &SpawnBurst_Instantaneous(Spawn Count=12, Use Spawn Probability=false, UseLoopCountLimit=false)
  ]
  particleSpawnScript [
   &InitializeParticle(
    AdjustAlpha=true
    AdjustHue=true
    AdjustSaturation=true
    AdjustValue=true
    Color=<rgba 0 0.0833 1 0.666>
    Color Mode=::NewEnumerator3
    Lifetime Max=4
    Lifetime Min=3
    Lifetime Mode=::NewEnumerator1
    Sprite Size Mode=::NewEnumerator1
    Uniform Sprite Size Max=150
    Uniform Sprite Size Min=100
    UsePositionOffset=false
    Write Material Random=true
   )
   &ShapeLocation(
    MaskSpawnGroup=false
    Override Seed=false
    Rotation Mode=::NewEnumerator5
    Scale Mode:ENiagara_ScaleMode=::NewEnumerator1
    Shape Origin=&TransformPosition(Position=Linked() {linkedVariable {name=System.ArrowTipPosition, type=Vector3f}}, Source Space=::Local, Destination Space=::Simulation)
    Sphere Radius=10
   )
   &AddVelocity(
    Velocity=&Multiply_Vector(A=$Particles.ShapeLocation.ShapeVector, B=<0.75 0.75 0.333>)
    Velocity Speed Scale=&RandomRangeFloat(Minimum=44, Maximum=66)
   )
  ]
  particleUpdateScript [
   &Drag(Drag=&RandomRangeFloat(Minimum=2, Maximum=4), Use Linear Drag=true, Use Rotational Drag=false)
   &GravityForce(Gravity=&RandomRangeVector(Minimum=<0 0 -150>, Maximum=<0 0 -200>))
   &CurlNoiseForce(Noise Frequency=10, Noise Strength=100, Pan Noise=false) {module=CurlNoiseForce001}
   &ParticleState()
   &ScaleSpriteSize(
    Uniform Curve Scale=&RandomRangeFloat(Minimum=0.75, Maximum=1.25)
    Uniform Curve Sprite Scale=DataInterface() {propertyValues {Curve {keys [
        {arriveTangent=2.9868, time=0, value=0}
        {arriveTangent=-1.2213, interpMode=RCIM_Cubic, leaveTangent=-1.2213, tangentMode=RCTM_User, time=0.1, value=0.833}
        {arriveTangent=-0.8542, leaveTangent=-0.0405, tangentMode=RCTM_User, time=0.7, value=0}
       ]}}}
   )
   &DynamicMaterialParameters(
    Index 0 Param 1=0.8
    Index 0 Param 2=1
    Index 0 Param 3=1
    Index 1 Param 1=150
    Index 1 Param 4=1
    Module.Param0WriteEnabled=true
    Module.Param0WriteEnabled1=true
    Module.Param1WriteEnabled=true
    Module.Param1WriteEnabled1=true
    Module.Param2WriteEnabled=true
    Module.Param2WriteEnabled1=true
    Module.Param3WriteEnabled=false
    Module.Param3WriteEnabled1=true
    Write Parameter Index 1=true
   )
   &ScaleColor(
    Scale RGB=&VectorFromFloat(Value=&Multiply_Float(
      A=$System.TimeOfDayBrightnessMultiplier
      B=&FloatFromCurve(FloatCurve=DataInterface() {propertyValues {Curve {keys [{arriveTangent=-523.03, interpMode=RCIM_Cubic, leaveTangent=-523.03, tangentMode=RCTM_User, time=0, value=30}, <at 0.1 1>]}}})
     ))
    ScaleA=true
    ScaleRGB=true
   ) {module=ScaleColor002}
   &SolveForcesAndVelocity(Clamp Velocity=false, Limit Acceleration=false)
  ]
  rendererClasses [&NiagaraSpriteRendererProperties]
  renderers [{rendererClass=&NiagaraSpriteRendererProperties}]
 }
 FireworkFlare_Explosion_Final_Parent001 {
  bEnabled=false
  emitterUpdateScript [
   &EmitterState(Life Cycle Mode=::NewEnumerator1, Loop Behavior=::NewEnumerator1, Loop Delay=2.6, Loop Duration=0.4, UseLoopDelay=true)
   &SpawnBurst_Instantaneous(Spawn Count=8, Use Spawn Probability=false, UseLoopCountLimit=false)
  ]
  particleSpawnScript [
   &InitializeParticle(
    AdjustAlpha=true
    AdjustHue=true
    AdjustSaturation=true
    AdjustValue=true
    Color=<rgba 0 0.0833 1 0.666>
    Color Mode=::NewEnumerator3
    Lifetime Max=4
    Lifetime Min=3
    Lifetime Mode=::NewEnumerator1
    Sprite Size Mode=::NewEnumerator1
    Uniform Sprite Size Max=150
    Uniform Sprite Size Min=100
    UsePositionOffset=false
    Write Material Random=true
   )
   &ShapeLocation(
    MaskSpawnGroup=false
    Override Seed=false
    Rotation Mode=::NewEnumerator5
    Scale Mode:ENiagara_ScaleMode=::NewEnumerator1
    Shape Origin=&TransformPosition(Position=Linked() {linkedVariable {name=System.ArrowTipPosition, type=Vector3f}}, Source Space=::Local, Destination Space=::Simulation)
    Sphere Radius=12
   )
   &AddVelocity(
    Velocity=&Multiply_Vector(A=$Particles.ShapeLocation.ShapeVector, B=<0.75 0.75 0.333>)
    Velocity Speed Scale=&RandomRangeFloat(Minimum=44, Maximum=66)
   )
  ]
  particleUpdateScript [
   &Drag(Drag=&RandomRangeFloat(Minimum=2, Maximum=4), Use Linear Drag=true, Use Rotational Drag=false)
   &GravityForce(Gravity=&RandomRangeVector(Minimum=<0 0 -150>, Maximum=<0 0 -200>))
   &CurlNoiseForce(Noise Frequency=10, Noise Strength=100, Pan Noise=false) {module=CurlNoiseForce001}
   &ParticleState()
   &ScaleSpriteSize(
    Uniform Curve Scale=&RandomRangeFloat(Minimum=0.75, Maximum=1.25)
    Uniform Curve Sprite Scale=DataInterface() {propertyValues {Curve {keys [
        {arriveTangent=2.9868, time=0, value=0}
        {arriveTangent=-1.2213, interpMode=RCIM_Cubic, leaveTangent=-1.2213, tangentMode=RCTM_User, time=0.1, value=0.833}
        {arriveTangent=-0.8542, leaveTangent=-0.0405, tangentMode=RCTM_User, time=0.7, value=0}
       ]}}}
   )
   &DynamicMaterialParameters(
    Index 0 Param 1=0.8
    Index 0 Param 2=1
    Index 0 Param 3=1
    Index 1 Param 1=150
    Index 1 Param 4=1
    Module.Param0WriteEnabled=true
    Module.Param0WriteEnabled1=true
    Module.Param1WriteEnabled=true
    Module.Param1WriteEnabled1=true
    Module.Param2WriteEnabled=true
    Module.Param2WriteEnabled1=true
    Module.Param3WriteEnabled=false
    Module.Param3WriteEnabled1=true
    Write Parameter Index 1=true
   )
   &ScaleColor(
    Scale RGB=&VectorFromFloat(Value=&Multiply_Float(
      A=$System.TimeOfDayBrightnessMultiplier
      B=&FloatFromCurve(FloatCurve=DataInterface() {propertyValues {Curve {keys [{arriveTangent=-523.03, interpMode=RCIM_Cubic, leaveTangent=-523.03, tangentMode=RCTM_User, time=0, value=30}, <at 0.1 1>]}}})
     ))
    ScaleA=true
    ScaleRGB=true
   ) {module=ScaleColor002}
   &SolveForcesAndVelocity(Clamp Velocity=false, Limit Acceleration=false)
  ]
  rendererClasses [&NiagaraSpriteRendererProperties]
  renderers [{rendererClass=&NiagaraSpriteRendererProperties}]
 }
 FireworkFlare_Explosion_ParentWhite {
  emitterUpdateScript [
   &EmitterState(Life Cycle Mode=::NewEnumerator1, Loop Behavior=::NewEnumerator1, Loop Delay=1.5, Loop Duration=0.4, UseLoopDelay=false)
   &SpawnBurst_Instantaneous(Use Spawn Probability=false, UseLoopCountLimit=false)
  ]
  particleSpawnScript [
   &InitializeParticle(
    Color=<rgba 0 1.5 3 1>
    Lifetime=0.4
    Position Offset=<75 0 0>
    Sprite Size Mode=::NewEnumerator3
    Uniform Sprite Size=40
    UsePositionOffset=true
    Write Material Random=true
   )
   &AddVelocity(Rotation Coordinate Space=::World, Velocity=<0 0 750>) {module=AddVelocity002}
  ]
  particleUpdateScript [
   &Drag(Drag=&RandomRangeFloat(Minimum=5, Maximum=7), Use Linear Drag=true, Use Rotational Drag=false)
   &SolveForcesAndVelocity(Clamp Velocity=false, Limit Acceleration=false)
   &ParticleState()
   &ScaleSpriteSize(Scale Factor=&Vector2DFromFloat(Value=&FloatFromCurve(
      FloatCurve=DataInterface() {propertyValues {Curve {keys [
          {arriveTangent=6.6282, interpMode=RCIM_Cubic, leaveTangent=6.6282, tangentMode=RCTM_User, time=0, value=0}
          <at 0.1 0.4626>
          {arriveTangent=4.0254, interpMode=RCIM_Cubic, leaveTangent=4.0254, time=0.1078, value=0.2796}
          <at 0.15 0.6638>
          <at 0.2 0.2796>
          <at 0.25 0.5723>
          <at 0.3 0.2887>
          <at 0.35 0.8011>
          <at 0.4 0.2887>
          {arriveTangent=-2.7618, leaveTangent=-2.7618, time=0.45, value=1}
          {arriveTangent=1, leaveTangent=1, time=0.55, value=0.3}
          {arriveTangent=-0.75, leaveTangent=-0.75, time=0.65, value=1.15}
          <at 0.75 0.5>
          {arriveTangent=-6.1253, interpMode=RCIM_Cubic, leaveTangent=-6.1253, tangentMode=RCTM_User, time=0.8, value=1.4}
          <at 1 0>
         ]}}}
      Scale Curve=3
     ))) {module=ScaleSpriteSize002}
   &DynamicMaterialParameters(
    Index 0 Param 1=0.8
    Index 0 Param 2=1
    Index 0 Param 3=1
    Index 1 Param 1=150
    Index 1 Param 4=1
    Module.Param0WriteEnabled=true
    Module.Param0WriteEnabled1=true
    Module.Param1WriteEnabled=true
    Module.Param1WriteEnabled1=true
    Module.Param2WriteEnabled=true
    Module.Param2WriteEnabled1=true
    Module.Param3WriteEnabled=false
    Module.Param3WriteEnabled1=true
    Write Parameter Index 1=true
   )
   &ScaleColor(
    Scale Alpha=&FloatFromCurve(
     FloatCurve=DataInterface() {propertyValues {Curve {keys [
         {arriveTangent=-0.4729, interpMode=RCIM_Cubic, leaveTangent=-0.4729, tangentMode=RCTM_User, time=0.5, value=1}
         {arriveTangent=-3.1771, tangentMode=RCTM_User, time=1, value=0}
        ]}}}
     Scale Curve=&RandomRangeFloat(Minimum=0.75, Maximum=&FN_FrontEndFloatModifier(Gameplay=1.25, Front End=1))
    )
    Scale RGB=&VectorFromFloat(Value=&FloatFromCurve(
      FloatCurve=DataInterface() {propertyValues {Curve {keys [
          {arriveTangent=-7.2198, interpMode=RCIM_Cubic, leaveTangent=-7.2198, tangentMode=RCTM_User, time=0, value=3}
          {arriveTangent=-3.1771, tangentMode=RCTM_User, time=1, value=1}
         ]}}}
      Scale Curve=&RandomRangeFloat(Minimum=0.75, Maximum=&FN_FrontEndFloatModifier(Gameplay=1.25, Front End=1))
     ))
    ScaleA=true
    ScaleRGB=true
   )
  ]
  rendererClasses [&NiagaraSpriteRendererProperties]
  renderers [{rendererClass=&NiagaraSpriteRendererProperties}]
 }
 Fireworks_Mesh_FinalExplosion {
  emitterUpdateScript [
   &EmitterState()
   &SpawnBurst_Instantaneous(Spawn Count=3, Use Spawn Probability=false, UseLoopCountLimit=false)
  ]
  particleSpawnScript [
   &InitializeParticle(
    Color=<rgba 0 0.6 2 1>
    Lifetime=0.7
    Mesh Scale Mode=::NewEnumerator3
    Mesh Uniform Scale=0.03
    Position Offset=<75 0 125>
    UsePositionOffset=true
    Write Material Random=true
   ) {module=InitializeParticle001}
   &InitialMeshOrientation(Use Rotation Vector=false) {module=InitialMeshOrientation001}
  ]
  particleUpdateScript [
   &ParticleState()
   &ScaleColor(
    Scale Alpha=&FloatFromCurve(FloatCurve=DataInterface() {propertyValues {Curve {keys [
         {arriveTangent=-4.9376, interpMode=RCIM_Cubic, leaveTangent=7.1838, tangentMode=RCTM_User, time=0, value=1}
         {arriveTangent=-2.4774, interpMode=RCIM_Cubic, leaveTangent=-2.4774, tangentMode=RCTM_User, time=0.1389, value=1.3591}
         {arriveTangent=-0.8467, interpMode=RCIM_Cubic, leaveTangent=-0.8467, tangentMode=RCTM_User, time=0.4181, value=0.6281}
         {arriveTangent=-2.194, interpMode=RCIM_Cubic, leaveTangent=-2.1946, tangentMode=RCTM_User, time=1, value=0}
        ]}}})
    Scale RGB=&VectorFromFloat(Value=1)
    ScaleA=true
    ScaleRGB=true
   ) {module=ScaleColor002}
   &ScaleMeshSize(Scale Factor=&VectorFromFloat(Value=&FloatFromCurve(FloatCurve=DataInterface() {propertyValues {Curve {keys [
          {arriveTangent=4.6803, interpMode=RCIM_Cubic, leaveTangent=4.6803, tangentMode=RCTM_User, time=0, value=0}
          {arriveTangent=1.6276, interpMode=RCIM_Cubic, leaveTangent=1.6276, tangentMode=RCTM_User, time=0.1923, value=0.5575}
          {arriveTangent=0.2011, interpMode=RCIM_Cubic, leaveTangent=0.5033, tangentMode=RCTM_User, time=1, value=1}
         ]}}})))
   &GravityForce(Gravity=&RandomRangeVector(Minimum=<0 0 -65>, Maximum=<0 0 -100>))
   &AccelerationForce(Acceleration=&RandomRangeVector(Minimum=<0 0 -25>, Maximum=<0 0 -50>))
   &SolveForcesAndVelocity(Clamp Velocity=false, Limit Acceleration=false)
   &DynamicMaterialParameters(
    Index 0 Param 1=1
    Index 0 Param 2=1
    Index 0 Param 3=&FloatFromCurve(
     FloatCurve=DataInterface() {propertyValues {Curve {keys [
         {arriveTangent=3.0114, interpMode=RCIM_Cubic, leaveTangent=3.0114, tangentMode=RCTM_User, time=0, value=0}
         {arriveTangent=0.9456, interpMode=RCIM_Cubic, leaveTangent=0.9456, tangentMode=RCTM_User, time=0.4343, value=0.7445}
         {arriveTangent=0.3236, interpMode=RCIM_Cubic, leaveTangent=0.7406, tangentMode=RCTM_User, time=1, value=1}
        ]}}}
     Scale Curve=0.2
    )
    Index 0 Param 4=&FloatFromCurve(
     FloatCurve=DataInterface() {propertyValues {Curve {keys [
         {arriveTangent=3.4321, interpMode=RCIM_Cubic, leaveTangent=3.4321, tangentMode=RCTM_User, time=0, value=1}
         {arriveTangent=-1.7957, interpMode=RCIM_Cubic, leaveTangent=-1.7957, tangentMode=RCTM_User, time=0.1672, value=1.5}
         {arriveTangent=-0.286, interpMode=RCIM_Cubic, leaveTangent=-0.286, tangentMode=RCTM_User, time=1, value=0.5}
        ]}}}
     Scale Curve=0.3
    )
    Module.Param0WriteEnabled=true
    Module.Param1WriteEnabled=true
    Module.Param2WriteEnabled=true
    Module.Param3WriteEnabled=true
   )
  ]
  rendererClasses [&NiagaraMeshRendererProperties]
  renderers [{rendererClass=&NiagaraMeshRendererProperties, rendererIndex=0}]
 }
 Flare_With_Streak002 {
  emitterUpdateScript [
   &EmitterState()
   &SpawnBurst_Instantaneous(Use Spawn Probability=false, UseLoopCountLimit=false)
  ]
  particleSpawnScript [
   &InitializeParticle(
    Color=<rgba 0 1.5 3 1>
    Lifetime=0.4
    Position Offset=<75 0 125>
    Position Offset Coordinate Space=::Local
    Sprite Rotation Mode=::NewEnumerator2
    Sprite Size Mode=::NewEnumerator3
    Uniform Sprite Size=150
    UsePositionOffset=true
    Write Material Random=true
   )
   &DynamicMaterialParameters(
    Index 0 Param 1=&RandomRangeFloat(Minimum=0.4, Maximum=0.6)
    Index 0 Param 2=1
    Index 0 Param 3=1
    Index 0 Param 4=1
    Module.Param0WriteEnabled=true
    Module.Param1WriteEnabled=true
    Module.Param2WriteEnabled=true
    Module.Param3WriteEnabled=true
   )
   /SetVariables() {isSetParameters=true}
  ]
  particleUpdateScript [
   &ParticleState()
   &SolveForcesAndVelocity(Clamp Velocity=false, Limit Acceleration=false)
   &ScaleColor(
    Scale Alpha=&Multiply_Float(
     A=1
     B=&FloatFromCurve(
      FloatCurve=DataInterface() {propertyValues {Curve {keys [<at 0 0>,<at 0.2 1>,<at 0.8 1>,<at 1 0>]}}}
      Scale Curve=$System.TimeOfDayBrightnessMultiplier
     )
    )
    ScaleA=true
    ScaleRGB=true
   )
   &ScaleSpriteSize(Scale Factor=&Vector2DFromFloat(Value=&FloatFromCurve(
      FloatCurve=DataInterface() {propertyValues {Curve {keys [<at 0 1.5 RCIM_Cubic>,<at 1 0>]}}}
      Scale Curve=&FN_FrontEndFloatModifier(Gameplay=1, Front End=0.7)
     )))
   &CameraOffset(Camera Offset Amount=50)
   /SetVariables(Particles.Flash size:Vector2f=&Multiply_Vector2D_ByFloat(
     Float=&Multiply_Float(
      A=180
      B=&FloatFromCurve(
       FloatCurve=DataInterface() {propertyValues {Curve {keys [
           {arriveTangent=4.1785, interpMode=RCIM_Cubic, leaveTangent=4.1785, tangentMode=RCTM_User, time=0, value=0}
           {arriveTangent=0.9995, interpMode=RCIM_Cubic, leaveTangent=0.9995, tangentMode=RCTM_User, time=0.1378, value=0.6825}
           {arriveTangent=-0.0405, interpMode=RCIM_Cubic, leaveTangent=-0.0405, tangentMode=RCTM_User, time=0.62, value=1}
          ]}}}
       Scale Curve=&FN_FrontEndFloatModifier(Gameplay=0.9, Front End=0.7)
      )
     )
     Vector2D=<1.5 1>
    )) {isSetParameters=true}
   /SetVariables(Particles.Flash color:LinearColor=&MakeLinearColorFromVectorAndFloat(
     "Vector (RGB)"=&MakeVectorFromLinearColorRGB(Color=<rgba 0 0.5 1 1>)
     "Float (Alpha)"=&Multiply_Float(
      A=&Multiply_Float(
       A=1
       B=&FloatFromCurve(
        FloatCurve=DataInterface() {propertyValues {Curve {keys [<at 0 0>,<at 0.124 1>,<at 0.496 1>,<at 0.62 0>]}}}
        Scale Curve=$System.TimeOfDayBrightnessMultiplier
       )
      )
      B=1
     )
    )) {isSetParameters=true}
   &FN_Light_Attributes(
    Color=$Particles.Color
    Create Color Attribute=true
    Create Exponent Attribute=true
    Create Position Attribute=true
    Create Radius Attribute=true
    Exponent=4
    Position=<75 0 50>
    Radius=250
    Scale Color=&FloatFromCurve(FloatCurve=DataInterface() {propertyValues {Curve {keys [
         {interpMode=RCIM_Cubic, tangentMode=RCTM_User, time=0, value=1}
         {arriveTangent=-2.6287, interpMode=RCIM_Cubic, leaveTangent=-2.6287, tangentMode=RCTM_User, time=1, value=0}
        ]}}})
   )
  ]
  rendererClasses [&NiagaraSpriteRendererProperties, &NiagaraLightRendererProperties]
  renderers [
   {rendererClass=&NiagaraSpriteRendererProperties}
   {rendererClass=&NiagaraSpriteRendererProperties, rendererIndex=1}
   {rendererClass=&NiagaraLightRendererProperties, rendererIndex=2}
  ]
 }
 Impact_Streaks {
  emitterUpdateScript [
   &EmitterState(Life Cycle Mode=::NewEnumerator1, Loop Behavior=::NewEnumerator1, Loop Duration=1, UseLoopDelay=false)
   &SpawnBurst_Instantaneous(
    Spawn Count=&RandomRangeInt(Minimum=8, Maximum=12)
    Use Spawn Probability=false
    UseLoopCountLimit=false
   )
  ]
  particleSpawnScript [
   &InitializeParticle(
    Color=<rgba 1.1286 8.5084 10 0.25>
    Lifetime Max=0.35
    Lifetime Min=0.12
    Lifetime Mode=::NewEnumerator1
    Position Offset=<75 0 0>
    Sprite Size Max=<10 35>
    Sprite Size Min=<5 15>
    Sprite Size Mode=::NewEnumerator2
    UsePositionOffset=true
    Write Material Random=true
   )
   &SubUVAnimation(End Frame=3, SubUV Animation Mode:ENiagara_SubUVLookupMode=::NewEnumerator2)
   &AddVelocity(
    Cone Axis=<0 0 1>
    Module.UseConeFalloff=false
    Module.UseVelDistribution=false
    Override Seed=false
    Velocity Mode=::NewEnumerator2
    Velocity Speed=&RandomRangeFloat(Minimum=50, Maximum=150)
   )
  ]
  particleUpdateScript [
   &SolveForcesAndVelocity(Clamp Velocity=false, Limit Acceleration=false)
   &ParticleState()
   &ScaleColor(
    Scale Alpha=&FloatFromCurve(FloatCurve=DataInterface() {propertyValues {Curve {keys [{arriveTangent=-3.3102, interpMode=RCIM_Cubic, leaveTangent=-3.3102, tangentMode=RCTM_User, time=0, value=1}, <at 1 0 RCIM_Cubic>]}}})
    Scale RGB=&MakeVectorFromLinearColorRGB(Color=&ColorFromCurve(ColorCurve=DataInterface() {propertyValues {
        AlphaCurve {keys []}
        BlueCurve {keys [<at 0 10 RCIM_Cubic>,<at 0.125 1>,<at 0.75 0 RCIM_Cubic>]}
        GreenCurve {keys [<at 0 10 RCIM_Cubic>,<at 0.125 1>,<at 0.75 0 RCIM_Cubic>]}
        RedCurve {keys [<at 0 10 RCIM_Cubic>,<at 0.125 1>,<at 0.75 0 RCIM_Cubic>]}
       }}))
    ScaleA=true
    ScaleRGB=true
   )
   &ScaleSpriteSize(
    Non-Uniform Curve Sprite Scale=DataInterface() {propertyValues {
      XCurve {keys [
        {arriveTangent=0.9067, interpMode=RCIM_Cubic, leaveTangent=0.9067, tangentMode=RCTM_User, time=0, value=0.8}
        {arriveTangentWeight=0.2034, interpMode=RCIM_Cubic, leaveTangentWeight=0.2034, tangentMode=RCTM_User, time=0.5, value=1}
        {arriveTangent=-3.1509, interpMode=RCIM_Cubic, leaveTangent=-3.1509, tangentMode=RCTM_User, time=1, value=0.1}
       ]}
      YCurve {keys [
        {arriveTangent=3.7985, interpMode=RCIM_Cubic, leaveTangent=3.7985, tangentMode=RCTM_User, time=0, value=0}
        {arriveTangent=2, interpMode=RCIM_Cubic, leaveTangent=2, time=0.5, value=1.6}
        <at 1 2 RCIM_Cubic>
       ]}
     }}
    Scale Sprite Size Mode=::NewEnumerator3
   )
  ]
  rendererClasses [&NiagaraSpriteRendererProperties]
  renderers [{rendererClass=&NiagaraSpriteRendererProperties}]
 }
 RibbonSmoke_Explosion002 {
  emitterUpdateScript [
   &EmitterState(
    Enable Distance Culling=true
    Life Cycle Mode=::NewEnumerator1
    Loop Behavior=::NewEnumerator1
    Loop Duration=2
    MaxDistance=2500000
    Module.bUseMaxDistance=true
    Module.bUseMinDistance=false
    Scalability Mode=::NewEnumerator1
    UseLoopDelay=false
   )
   &SpawnParticlesFromOtherEmitter(
    Attribute Reader=DataInterface() {propertyValues {EmitterBinding {bindingMode=Other, emitterName=FireworkFlare_Explosion_ParentWhite}}}
    Spawn Rate=150
    Spawn Rate Per Particle Cap=150
    Use Maximum Age=false
    Use Minimum Age=false
   )
  ]
  particleSpawnScript [
   &InitializeRibbon(
    Color=<rgba 1 1 1 2>
    Lifetime=0.4
    Position Mode:ENiagara_PositionInitializationMode=::NewEnumerator1
    Ribbon Width=&RandomRangeFloat(Minimum=1, Maximum=2)
    UsePositionOffset=false
    Write Material Random=true
   ) {module=InitializeRibbon001}
   &SampleParticlesFromOtherEmitter(
    Age Sampling:ENiagara_DirectReadSamplingMode=::NewEnumerator0
    Applied Velocity Scale=0
    Apply Sampled Alpha=false
    Color Sampling=::NewEnumerator0
    Lifetime Application:ENiagara_LifetimeInheritanceOptions=::NewEnumerator1
    Lifetime Sampling:ENiagara_DirectReadSamplingMode=::NewEnumerator0
    Particle ID Sampling=::NewEnumerator1
    Reject Particle If Sampled Index Is Invalid=false
    Sprite Size Sampling:ENiagara_DirectReadSamplingMode=::NewEnumerator0
    Velocity Sampling=::NewEnumerator0
   )
  ]
  particleUpdateScript [
   &ParticleState()
   &DynamicMaterialParameters(
    Index 0 Param 1=-0.2
    Index 0 Param 4=&FloatFromCurve(FloatCurve=DataInterface() {propertyValues {Curve {keys [
         <at 0 1 RCIM_Cubic>
         {arriveTangent=-0.3679, interpMode=RCIM_Cubic, leaveTangent=-0.3679, tangentMode=RCTM_User, time=1.6, value=0.7}
         {arriveTangent=-0.0204, tangentMode=RCTM_User, time=3, value=0}
        ]}}})
    Module.Param0WriteEnabled=true
    Module.Param1WriteEnabled=true
    Module.Param2WriteEnabled=true
    Module.Param3WriteEnabled=true
   )
   &ScaleRibbonWidth(Ribbon Width Scale=&FloatFromCurve(
     FloatCurve=DataInterface() {propertyValues {Curve {keys [{arriveTangent=-1.2246, interpMode=RCIM_Cubic, leaveTangent=-1.2246, tangentMode=RCTM_User, time=0.1, value=1}, <at 2 0>]}}}
     Scale Curve=7
    )) {module=ScaleRibbonWidth001}
   /Niagara/Modules/Update/Forces/CurlNoiseForce(
    Noise Frequency=1
    Noise Strength=&Multiply_Float(
     A=1727.7727
     B=&FloatFromCurve(
      FloatCurve=DataInterface() {propertyValues {Curve {keys [
          {arriveTangent=1.9891, interpMode=RCIM_Cubic, leaveTangent=1.9891, tangentMode=RCTM_User, time=0, value=0}
          {interpMode=RCIM_Cubic, tangentMode=RCTM_User, time=0.75, value=1}
          {tangentMode=RCTM_User, time=1, value=0}
         ]}}}
      Scale Curve=0.2
     )
    )
    Pan Noise=true
   )
   &ScaleColor(
    Scale Alpha=&FloatFromCurve(FloatCurve=DataInterface() {propertyValues {Curve {keys [
         {arriveTangent=9.0725, interpMode=RCIM_Cubic, leaveTangent=9.0725, tangentMode=RCTM_User, time=0, value=0.01}
         {arriveTangent=-2.5249, interpMode=RCIM_Cubic, leaveTangent=-2.5249, tangentMode=RCTM_User, time=0.1, value=1}
         {arriveTangent=-0.0405, interpMode=RCIM_Cubic, leaveTangent=-0.0405, tangentMode=RCTM_User, time=1, value=0}
        ]}}})
    Scale RGB=&VectorFromFloat(Value=1)
    ScaleA=true
    ScaleRGB=true
   )
   &SolveForcesAndVelocity(Clamp Velocity=false, Limit Acceleration=false)
  ]
  rendererClasses [&NiagaraRibbonRendererProperties]
  renderers [{rendererClass=&NiagaraRibbonRendererProperties}]
 }
 Spark_High {
  simTarget=GPUComputeSim
  emitterUpdateScript [
   &EmitterState()
   &SpawnBurst_Instantaneous(Spawn Count=250, Use Spawn Probability=false, UseLoopCountLimit=false)
  ]
  particleSpawnScript [
   &InitializeParticle(
    AdjustAlpha=false
    AdjustHue=true
    AdjustSaturation=false
    AdjustValue=false
    Color=<rgba 0 1.5 3 1>
    Color Mode=::NewEnumerator3
    Hue Shift Range:Vector2f=<-0.2 0>
    Lifetime Max=1.6
    Lifetime Min=0.7
    Lifetime Mode=::NewEnumerator1
    Position Offset=<75 0 125>
    Position Offset Coordinate Space=::Local
    Sprite Size Mode=::NewEnumerator1
    Uniform Sprite Size Max=3
    Uniform Sprite Size Min=1
    UsePositionOffset=true
    Write Material Random=true
   )
   &ShapeLocation(MaskSpawnGroup=false, Override Seed=false, Sphere Radius=22, UseOwnerScale=true)
   &AddVelocityFromPoint(
    Module.Use Falloff Distance=false
    Origin Offset=<75 0 125>
    Velocity Strength=&RandomRangeFloat(Minimum=200, Maximum=325)
   )
   /Niagara/Modules/Update/Forces/V2/CurlNoiseForce(Noise Frequency=65, Noise Strength=450, Pan Noise=true) {enabled=false}
   &ApplyInitialForces()
  ]
  particleUpdateScript [
   &ParticleState()
   &GravityForce(Gravity=&RandomRangeVector(Minimum=<0 0 -250>, Maximum=<0 0 -350>))
   &Drag(Drag=&RandomRangeFloat(Minimum=6, Maximum=11), Use Linear Drag=true, Use Rotational Drag=true)
   &SolveForcesAndVelocity(Clamp Velocity=false, Limit Acceleration=false)
   &ScaleSpriteSize(Uniform Curve Sprite Scale=DataInterface() {propertyValues {Curve {keys [{tangentMode=RCTM_User, time=0, value=1}, {tangentMode=RCTM_User, time=1, value=0}]}}})
   &ScaleSpriteSizeBySpeed(Max Scale Factor=<0.2 5>, Min Scale Factor=<1 0.5>)
   &ScaleColor(
    Scale Alpha=&FloatFromCurve(FloatCurve=DataInterface() {propertyValues {Curve {keys [<at 0 1>,<at 0.3 0.25>,<at 1 0>]}}})
    Scale RGB=&VectorFromFloat(Value=$System.TimeOfDayBrightnessMultiplier)
    ScaleA=true
    ScaleRGB=true
   )
  ]
  rendererClasses [&NiagaraSpriteRendererProperties]
  renderers [{rendererClass=&NiagaraSpriteRendererProperties}]
 }
 Spark_Low {
  emitterUpdateScript [
   &EmitterState()
   &SpawnBurst_Instantaneous(Spawn Count=150, Use Spawn Probability=false, UseLoopCountLimit=false)
  ]
  particleSpawnScript [
   &InitializeParticle(
    AdjustAlpha=false
    AdjustHue=true
    AdjustSaturation=false
    AdjustValue=false
    Color=<rgba 0 1.5 3 1>
    Color Mode=::NewEnumerator3
    Hue Shift Range:Vector2f=<-0.2 0>
    Lifetime Max=1.6
    Lifetime Min=0.7
    Lifetime Mode=::NewEnumerator1
    Position Offset=<75 0 125>
    Position Offset Coordinate Space=::Local
    Sprite Size Mode=::NewEnumerator1
    Uniform Sprite Size Max=3
    Uniform Sprite Size Min=1
    UsePositionOffset=true
    Write Material Random=true
   )
   &ShapeLocation(MaskSpawnGroup=false, Override Seed=false, Sphere Radius=22, UseOwnerScale=true)
   &AddVelocityFromPoint(
    Module.Use Falloff Distance=false
    Origin Offset=<75 0 125>
    Velocity Strength=&RandomRangeFloat(Minimum=200, Maximum=325)
   )
   &CurlNoiseForce(Noise Frequency=65, Noise Strength=450, Pan Noise=true) {enabled=false}
   &ApplyInitialForces()
  ]
  particleUpdateScript [
   &ParticleState()
   &GravityForce(Gravity=&RandomRangeVector(Minimum=<0 0 -250>, Maximum=<0 0 -350>))
   &Drag(Drag=&RandomRangeFloat(Minimum=6, Maximum=11), Use Linear Drag=true, Use Rotational Drag=true)
   &SolveForcesAndVelocity(Clamp Velocity=false, Limit Acceleration=false)
   &ScaleSpriteSize(Uniform Curve Sprite Scale=DataInterface() {propertyValues {Curve {keys [{tangentMode=RCTM_User, time=0, value=1}, {tangentMode=RCTM_User, time=1, value=0}]}}})
   &ScaleSpriteSizeBySpeed(Max Scale Factor=<0.2 5>, Min Scale Factor=<1 0.5>)
   &ScaleColor(
    Scale Alpha=&FloatFromCurve(FloatCurve=DataInterface() {propertyValues {Curve {keys [<at 0 1>,<at 0.3 0.25>,<at 1 0>]}}})
    Scale RGB=&VectorFromFloat(Value=$System.TimeOfDayBrightnessMultiplier)
    ScaleA=true
    ScaleRGB=true
   )
  ]
  rendererClasses [&NiagaraSpriteRendererProperties]
  renderers [{rendererClass=&NiagaraSpriteRendererProperties}]
 }
 TrailEmbers_Explosion002 {
  emitterUpdateScript [
   &EmitterState(Life Cycle Mode=::NewEnumerator1, Loop Behavior=::NewEnumerator1, Loop Duration=2, UseLoopDelay=false)
   &SpawnParticlesFromOtherEmitter(
    Attribute Reader=DataInterface() {propertyValues {EmitterBinding {bindingMode=Other, emitterName=FireworkFlare_Explosion_ParentWhite}}}
    Maximum Age Spawn Threshold=2
    Minimum Age Spawn Threshold=0
    Spawn Rate=125
    Spawn Rate Per Particle Cap=125
    Use Maximum Age=false
    Use Minimum Age=false
   )
  ]
  particleSpawnScript [
   &InitializeParticle(
    Color Mode=::NewEnumerator0
    Lifetime Max=0.4
    Lifetime Min=0.25
    Lifetime Mode=::NewEnumerator1
    Sprite Rotation Mode=::NewEnumerator2
    Sprite Size Mode=::NewEnumerator1
    Uniform Sprite Size Max=2
    Uniform Sprite Size Min=1
    UsePositionOffset=false
    Write Material Random=true
   )
   &SampleParticlesFromOtherEmitter(
    Applied Velocity Scale:NiagaraFloat=&FloatFromCurve(
     FloatCurve=DataInterface() {propertyValues {Curve {keys [{arriveTangent=-1.801, interpMode=RCIM_Cubic, leaveTangent=-1.801, tangentMode=RCTM_User, time=0, value=1}, <at 1 0>]}}}
     Scale Curve=0.5
    )
    Apply Sampled Alpha=false
    Color Sampling=::NewEnumerator0
    Maximum Sampled Particle Age=2
    Minimum Sampled Particle Age=0
    Module.MaxSampledAgeCutoff=true
    Module.MinSampledAgeCutoff=true
    Reject Particle By Sampled Age=true
    Velocity Sampling=::NewEnumerator0
   )
   &AddVelocity(
    Cone Angle=1
    Cone Axis=&MakeVector(X=&Sine(Normalized Angle=$Emitter.Age, Period=5, Scale=1, Bias=0), Y=0, Z=1)
    Module.UseConeFalloff=false
    Module.UseVelDistribution=false
    Override Seed=false
    Velocity Mode=::NewEnumerator2
    Velocity Speed=5
   )
   &SphereLocation(
    MaskSpawnGroup=false
    Override Seed=false
    Sphere Radius=2
    Surface Only=false
    UseNonUniformScale=false
    UseOffset=false
    UseSphereOrientation=false
   )
  ]
  particleUpdateScript [
   &ParticleState()
   &Drag(Drag=&RandomRangeFloat(Minimum=12, Maximum=15), Use Linear Drag=true, Use Rotational Drag=false)
   &CurlNoiseForce(
    Noise Frequency=&RandomRangeFloat(Minimum=0.25, Maximum=1)
    Noise Strength=&RandomRangeFloat(Minimum=200, Maximum=500)
    Pan Noise=true
    Pan Noise Field=&RandomRangeVector(Minimum=<0 0 0>, Maximum=<1 1 1>)
   )
   &SolveForcesAndVelocity(Clamp Velocity=false, Limit Acceleration=false)
   &ScaleSpriteSize(
    Uniform Curve Scale=0.5
    Uniform Curve Sprite Scale=DataInterface() {propertyValues {Curve {keys [
        {arriveTangent=2.9868, interpMode=RCIM_Cubic, leaveTangent=2.9868, tangentMode=RCTM_User, time=0, value=1.125}
        <at 0.4 2.3>
        {arriveTangent=-0.0405, interpMode=RCIM_Cubic, leaveTangent=-0.0405, tangentMode=RCTM_User, time=1, value=1.4}
       ]}}}
   )
   &ScaleColor(
    Scale Alpha=&FloatFromCurve(FloatCurve=DataInterface() {propertyValues {Curve {keys [{arriveTangent=-1.933, interpMode=RCIM_Cubic, leaveTangent=-1.933, tangentMode=RCTM_User, time=0, value=1}, <at 1 0>]}}})
    Scale RGB=&VectorFromFloat(Value=&FloatFromCurve(FloatCurve=DataInterface() {propertyValues {Curve {keys [{arriveTangent=-180.164, interpMode=RCIM_Cubic, leaveTangent=-180.164, tangentMode=RCTM_User, time=0, value=50}, <at 1 1>]}}}))
    ScaleA=true
    ScaleRGB=true
   )
  ]
  rendererClasses [&NiagaraSpriteRendererProperties]
  renderers [{rendererClass=&NiagaraSpriteRendererProperties}]
 }
 TrailEmbers_FinalExplosion {
  bEnabled=false
  simTarget=GPUComputeSim
  emitterUpdateScript [
   &EmitterState()
   &SpawnParticlesFromOtherEmitter(
    Attribute Reader=DataInterface() {propertyValues {EmitterBinding {bindingMode=Other, emitterName=FireworkFlare_Explosion_Final_Parent}}}
    Maximum Age Spawn Threshold=15
    Minimum Age Spawn Threshold=0
    Spawn Rate=&SelectFloatFromArray(Direct Array Index=$Engine.QualityLevel, Float Selection Array=DataInterface() {propertyValues {FloatData [50,75,100,125,125]}})
    Spawn Rate Per Particle Cap=125
    Use Maximum Age=false
    Use Minimum Age=false
   )
  ]
  particleSpawnScript [
   &PartitionParticles(Number Of Partitions=4, Partition Mode:ENiagara_PartitionMode=::NewEnumerator5)
   &InitializeParticle(
    Color Mode=::NewEnumerator0
    Lifetime Max=&MergeHelper_FloatSwitch(0=0.3, 1=1, 2=?, 3=?, Index=$Particles.PartitionParticles.Partition, Num=::NewEnumerator0)
    Lifetime Min:NiagaraFloat=&MergeHelper_FloatSwitch(0=0.2, 1=0.7, 2=?, 3=?, Index=$Particles.PartitionParticles.Partition, Num=::NewEnumerator0)
    Lifetime Mode=::NewEnumerator1
    Sprite Rotation Mode=::NewEnumerator2
    Sprite Size Mode=::NewEnumerator3
    Uniform Sprite Size=&MergeHelper_FloatSwitch(0=0.75, 1=3, 2=?, 3=?, Index=$Particles.PartitionParticles.Partition, Num=::NewEnumerator0)
    UsePositionOffset=false
    Write Material Random=true
   )
   &SampleParticlesFromOtherEmitter(
    Applied Velocity Scale:NiagaraFloat=&FloatFromCurve(
     FloatCurve=DataInterface() {propertyValues {Curve {keys [{arriveTangent=-1.801, interpMode=RCIM_Cubic, leaveTangent=-1.801, tangentMode=RCTM_User, time=0, value=1}, <at 1 0>]}}}
     Scale Curve=0.5
    )
    Apply Sampled Alpha=false
    Color Sampling=::NewEnumerator0
    Velocity Sampling=::NewEnumerator0
   )
   &SphereLocation(
    MaskSpawnGroup=false
    Override Seed=false
    Sphere Radius=2
    Surface Only=false
    UseNonUniformScale=false
    UseOffset=false
    UseSphereOrientation=false
   )
  ]
  particleUpdateScript [
   &ParticleState()
   &Drag(Drag=&RandomRangeFloat(Minimum=4, Maximum=6), Use Linear Drag=true, Use Rotational Drag=false)
   &CurlNoiseForce(
    Noise Frequency=&RandomRangeFloat(Minimum=0.25, Maximum=1)
    Noise Strength=&RandomRangeFloat(Minimum=100, Maximum=200)
    Pan Noise=true
    Pan Noise Field=&RandomRangeVector(Minimum=<0 0 0>, Maximum=<1 1 1>)
   )
   &SolveForcesAndVelocity(Clamp Velocity=false, Limit Acceleration=false)
   &ScaleSpriteSize(
    Uniform Curve Scale=0.5
    Uniform Curve Sprite Scale=DataInterface() {propertyValues {Curve {keys [
        {arriveTangent=2.9868, interpMode=RCIM_Cubic, leaveTangent=2.9868, tangentMode=RCTM_User, time=0, value=1.125}
        <at 0.4 2.3>
        {arriveTangent=-0.0405, interpMode=RCIM_Cubic, leaveTangent=-0.0405, tangentMode=RCTM_User, time=1, value=1.4}
       ]}}}
   )
   &ScaleColor(
    Scale Alpha=&FloatFromCurve(
     FloatCurve=DataInterface() {propertyValues {Curve {keys [{arriveTangent=-1.933, interpMode=RCIM_Cubic, leaveTangent=-1.933, tangentMode=RCTM_User, time=0, value=1}, <at 1 0>]}}}
     Scale Curve=&MergeHelper_FloatSwitch(0=20, 1=0.25, 2=?, 3=?, Index=$Particles.PartitionParticles.Partition, Num=::NewEnumerator0)
    )
    Scale RGB=&VectorFromFloat(Value=&FloatFromCurve(
      FloatCurve=DataInterface() {propertyValues {Curve {keys [{arriveTangent=-180.164, interpMode=RCIM_Cubic, leaveTangent=-180.164, tangentMode=RCTM_User, time=0, value=50}, <at 1 1>]}}}
      Scale Curve=&MergeHelper_FloatSwitch(0=20, 1=0.25, 2=?, 3=?, Index=$Particles.PartitionParticles.Partition, Num=::NewEnumerator0)
     ))
    ScaleA=true
    ScaleRGB=true
   )
  ]
  rendererClasses [&NiagaraSpriteRendererProperties]
  renderers [{rendererClass=&NiagaraSpriteRendererProperties}]
 }
 TrailEmbers_FinalExplosion001 {
  bEnabled=false
  emitterUpdateScript [
   &EmitterState(Life Cycle Mode=::NewEnumerator1, Loop Behavior=::NewEnumerator1, Loop Delay=2.6, Loop Duration=7, UseLoopDelay=true)
   &SpawnParticlesFromOtherEmitter(
    Attribute Reader=DataInterface() {propertyValues {EmitterBinding {bindingMode=Other, emitterName=FireworkFlare_Explosion_Final_Parent001}}}
    Maximum Age Spawn Threshold=2
    Minimum Age Spawn Threshold=0
    Spawn Rate=&SelectFloatFromArray(Direct Array Index=$Engine.QualityLevel, Float Selection Array=DataInterface() {propertyValues {FloatData [50,75,100,125,125]}})
    Spawn Rate Per Particle Cap=125
    Use Maximum Age=false
    Use Minimum Age=false
   )
  ]
  particleSpawnScript [
   &InitializeParticle(
    Color Mode=::NewEnumerator0
    Lifetime Max=1
    Lifetime Min=0.7
    Lifetime Mode=::NewEnumerator1
    Sprite Rotation Mode=::NewEnumerator2
    Sprite Size Mode=::NewEnumerator1
    Uniform Sprite Size Max=3
    Uniform Sprite Size Min=1
    UsePositionOffset=false
    Write Material Random=true
   )
   &SampleParticlesFromOtherEmitter(
    Applied Velocity Scale:NiagaraFloat=&FloatFromCurve(
     FloatCurve=DataInterface() {propertyValues {Curve {keys [{arriveTangent=-1.801, interpMode=RCIM_Cubic, leaveTangent=-1.801, tangentMode=RCTM_User, time=0, value=1}, <at 1 0>]}}}
     Scale Curve=0.5
    )
    Apply Sampled Alpha=false
    Color Sampling=::NewEnumerator0
    Maximum Sampled Particle Age=7
    Minimum Sampled Particle Age=0
    Module.MaxSampledAgeCutoff=true
    Module.MinSampledAgeCutoff=true
    Reject Particle By Sampled Age=true
    Velocity Sampling=::NewEnumerator0
   )
   &SphereLocation(
    MaskSpawnGroup=false
    Override Seed=false
    Sphere Radius=2
    Surface Only=false
    UseNonUniformScale=false
    UseOffset=false
    UseSphereOrientation=false
   )
  ]
  particleUpdateScript [
   &ParticleState()
   &Drag(Drag=&RandomRangeFloat(Minimum=4, Maximum=6), Use Linear Drag=true, Use Rotational Drag=false)
   &CurlNoiseForce(
    Noise Frequency=&RandomRangeFloat(Minimum=0.25, Maximum=1)
    Noise Strength=&RandomRangeFloat(Minimum=100, Maximum=200)
    Pan Noise=true
    Pan Noise Field=&RandomRangeVector(Minimum=<0 0 0>, Maximum=<1 1 1>)
   )
   &SolveForcesAndVelocity(Clamp Velocity=false, Limit Acceleration=false)
   &ScaleSpriteSize(
    Uniform Curve Scale=0.5
    Uniform Curve Sprite Scale=DataInterface() {propertyValues {Curve {keys [
        {arriveTangent=2.9868, interpMode=RCIM_Cubic, leaveTangent=2.9868, tangentMode=RCTM_User, time=0, value=1.125}
        <at 0.4 2.3>
        {arriveTangent=-0.0405, interpMode=RCIM_Cubic, leaveTangent=-0.0405, tangentMode=RCTM_User, time=1, value=1.4}
       ]}}}
   )
   &ScaleColor(
    Scale Alpha=&FloatFromCurve(FloatCurve=DataInterface() {propertyValues {Curve {keys [{arriveTangent=-1.933, interpMode=RCIM_Cubic, leaveTangent=-1.933, tangentMode=RCTM_User, time=0, value=1}, <at 1 0>]}}})
    Scale RGB=&VectorFromFloat(Value=&FloatFromCurve(
      FloatCurve=DataInterface() {propertyValues {Curve {keys [{arriveTangent=-180.164, interpMode=RCIM_Cubic, leaveTangent=-180.164, tangentMode=RCTM_User, time=0, value=50}, <at 1 1>]}}}
      Scale Curve=&FN_FrontEndFloatModifier(Gameplay=50, Front End=10)
     ))
    ScaleA=true
    ScaleRGB=true
   )
  ]
  rendererClasses [&NiagaraSpriteRendererProperties]
  renderers [{rendererClass=&NiagaraSpriteRendererProperties}]
 }
}
summary {
 systemName=NSScoopPluck_FireworkBlue1
 emitters [
  {
   bEnabled=false
   emitterName=FireworkFlare_Explosion_Final_Parent
   simTarget=GPUComputeSim
   rendererClasses [&NiagaraSpriteRendererProperties]
  }
  {
   bEnabled=false
   emitterName=TrailEmbers_FinalExplosion
   simTarget=GPUComputeSim
   rendererClasses [&NiagaraSpriteRendererProperties]
  }
  {bEnabled=false, emitterName=FireworkFlare_Explosion_Final_Parent001, rendererClasses [&NiagaraSpriteRendererProperties]}
  {bEnabled=false, emitterName=TrailEmbers_FinalExplosion001, rendererClasses [&NiagaraSpriteRendererProperties]}
  {emitterName=FireworkFlare_Explosion_ParentWhite, rendererClasses [&NiagaraSpriteRendererProperties]}
  {emitterName=TrailEmbers_Explosion002, rendererClasses [&NiagaraSpriteRendererProperties]}
  {emitterName=RibbonSmoke_Explosion002, rendererClasses [&NiagaraRibbonRendererProperties]}
  {emitterName=Dust002, rendererClasses [&NiagaraSpriteRendererProperties]}
  {
   emitterName=Flare_With_Streak002
   rendererClasses [&NiagaraSpriteRendererProperties, &NiagaraLightRendererProperties]
  }
  {emitterName=BG_Glow002, rendererClasses [&NiagaraSpriteRendererProperties]}
  {emitterName=Spark_High, simTarget=GPUComputeSim, rendererClasses [&NiagaraSpriteRendererProperties]}
  {emitterName=Spark_Low, rendererClasses [&NiagaraSpriteRendererProperties]}
  {emitterName=Fireworks_Mesh_FinalExplosion, rendererClasses [&NiagaraMeshRendererProperties]}
  {emitterName=Impact_Streaks, rendererClasses [&NiagaraSpriteRendererProperties]}
 ]
 userVariables [
  User.bIsFrontEnd {type=NiagaraBool}
  User.bIsFrontEndPreview {type=NiagaraBool}
 ]
}
