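# Niagara ground-impact effect for the ScoopPluck emote, combining bouncing rock debris, ground smoke, a radial smoke pulse and orange rim glows with light.
# A computed smoke center anchors rocks and pulse, while a height factor scales spawning, force and opacity.
# Water suppresses rocks and main smoke, and owner velocity gates a second smoke group.
# Front-end flags adjust visuals, the pulse varies by time of day and camera distance, and one glow follows slopes.
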NiagaraSystem /BRCosmetics/Animation/Game/MainPlayer/Emotes/ScoopPluck/FX/NS_Emote_ScoopPluck_Dust
dependencies {
 usedDataInterfaces [
  &NiagaraDataInterfaceCurve
  &NiagaraDataInterfaceCamera
  &NiagaraDataInterfaceColorCurve
  &NiagaraDataInterfaceCollisionQuery
  &NiagaraDataInterfaceDebugDraw
 ]
 usedDynamicInputs [
  &Multiply_Vector2D_ByFloat
  &ClampFloat
  &Divide_Float
  &Subtract_Float
 ]
 usedModules [
  &SystemState
  &self:CalcSmokeCenterPosition
  &self:SystemSpawnScript.SystemScriptSource.SystemScriptGraph.NiagaraNodeAssignment_0.SetVariables_#1
  &self:SystemSpawnScript.SystemScriptSource.SystemScriptGraph.NiagaraNodeAssignment_1.SetVariables_#2
 ]
}
emitters {
 GroundRocks {
  emitterUpdateScript [
   &EmitterState()
   &SpawnRate(
    SpawnRate=&Lerp_Float(
     A=0
     B=&RandomRangeFloat(Minimum=-2, Maximum=&Lerp_Float(A=-23.625, B=33.75, Alpha=$System.HeightFactor), Evaluation Type=::NewEnumerator1)
     Alpha=&MakeCustomFloatFromBool(Bool=&Boolean_OrOperation(Bool1=$User.bIsInWater, Bool2=$User.bIsFrontEndPreview), True Float=0, False Float=1)
    )
    Use Spawn Probability=false
   )
  ]
  particleSpawnScript [
   /SetVariables(Particles.MaterialRandom:NiagaraFloat=&RandomRangeFloat(Minimum=0, Maximum=1)) {isSetParameters=true}
   /Niagara/Modules/Spawn/Initialization/InitializeParticle(
    Color=&LerpLinearColors(StartColor=<rgba 0.65 0.5738 0.5193 1>, EndColor=<rgba 0.75 0.6437 0.5677 1>, LerpFactor=$Particles.MaterialRandom)
    Lifetime=&Lerp_Float(A=0.7, B=1, Alpha=$Particles.MaterialRandom)
    Module.Write Color=true
    Module.Write Lifetime=true
    Module.Write Mass=true
    Module.Write Material Random=false
    Module.Write Position=false
    Module.Write Scale=false
    Module.Write SpriteRotation=true
    Module.Write SpriteSize=true
    Position=&SimulationPosition()
    Sprite Rotation=&RandomRangeFloat(Minimum=0, Maximum=360)
    Sprite Size=&Vector2DFromFloat(Value=&RandomRangeFloat(Minimum=3, Maximum=7))
   )
   /SetVariables(Particles.Position:Vector3f=$System.SmokeCenterPositionWS) {isSetParameters=true}
   &SphereLocation(
    MaskSpawnGroup=false
    Non Uniform Scale=<1 1 0>
    Offset=<0 0 5>
    Override Seed=false
    Sphere Origin:Vector3f=$System.SmokeCenterPositionWS
    Sphere Radius=&Lerp_Float(A=25, B=25, Alpha=&Power(a=$Particles.MaterialRandom, b=1.5))
    Surface Only=false
    UseNonUniformScale=true
    UseOffset=true
    UseSphereOrientation=false
   )
   &AddVelocityFromPoint(
    Module.Use Falloff Distance=false
    Velocity Origin:Vector3f=$System.SmokeCenterPositionWS
    Velocity Strength=&Lerp_Float(A=180, B=220, Alpha=&Power(a=$Particles.MaterialRandom, b=3))
   )
   &AddVelocity(Velocity=<0 0 200>)
  ]
  particleUpdateScript [
   &ParticleState()
   &GravityForce()
   &Drag(Use Linear Drag=true, Use Rotational Drag=false)
   &SpriteRotationRate(Rotation Rate=&Lerp_Float(A=100, B=650, Alpha=$Particles.MaterialRandom))
   &ScaleSpriteSize(Scale Factor=&Vector2DFromFloat(Value=&FloatFromCurve(FloatCurve=DataInterface() {propertyValues {Curve {keys [
          {arriveTangent=35.6763, interpMode=RCIM_Cubic, leaveTangent=3.7835, tangentMode=RCTM_User, time=0, value=0}
          <at 0.7 1 RCIM_Cubic>
          {interpMode=RCIM_Cubic, leaveTangent=0.6903, time=1, value=0}
         ]}}})))
   &Collision(
    Advanced Aging Rate=1
    Age Colliding Particles=false
    Collision Enabled=&SetBoolByFloatComparison(A=$Particles.Age, B=0.1)
    Module.Control Rest State Via Surface Normal=false
    Module.EnableMaxCollisionCount=false
    Module.Randomize Collision Normal Vector=false
    Restitution=0.3
   )
   /SetVariables(Particles.MaxCollisions:NiagaraBool=&SetBoolByIntComparison(A=$Particles.Number Of Collisions, B=1, Comparison Type=::NewEnumerator0)) {isSetParameters=true}
   &SolveForcesAndVelocity(Clamp Velocity=false, Limit Acceleration=false)
   /SetVariables(Particles.SpriteRotation:NiagaraFloat=&MakeCustomFloatFromBool(Bool=$Particles.MaxCollisions, True Float=0, False Float=$Particles.SpriteRotation)) {isSetParameters=true}
   &ScaleColor(
    Scale RGB=&VectorFromFloat(Value=&MakeCustomFloatFromBool(Bool=$User.bIsFrontEndPreview, True Float=1.5, False Float=1))
    ScaleA=true
    ScaleRGB=true
   )
  ]
  rendererClasses [&NiagaraSpriteRendererProperties]
  renderers [{rendererClass=&NiagaraSpriteRendererProperties}]
 }
 GroundSmoke {
  emitterUpdateScript [
   &EmitterState()
   &SpawnRate(
    SpawnRate=&Lerp_Float(
     A=0
     B=&Lerp_Float(A=-50, B=50, Alpha=$System.HeightFactor)
     Alpha=&MakeCustomFloatFromBool(Bool=$User.bIsInWater, True Float=0, False Float=1)
    )
    Use Spawn Probability=false
   )
   &SpawnRate(
    SpawnGroup=1
    SpawnRate=&MakeCustomFloatFromBool(
     Bool=&SetBoolByFloatComparison(A=&VectorLength(Vector=$User.OwnerLocalVelocity), B=7, Comparison Type=::NewEnumerator6)
     True Float=7
     False Float=0
    )
    Use Spawn Probability=false
   ) {module=SpawnRate001}
  ]
  particleSpawnScript [
   /SetVariables(Particles.VisibilityTag=&SpawnGroup()) {isSetParameters=true}
   /Niagara/Modules/Spawn/Initialization/V2/InitializeParticle(
    Color=&MaskLinearColorBySpawnGroup(Masked Linear Color=<rgba 0.629 0.7403 1 0.1>, Passthrough Linear Color=<rgba 0.635 0.8149 0.919 0.3>)
    Lifetime=&MaskFloatBySpawnGroup(
     Masked Float=&RandomRangeFloat(Minimum=0.3, Maximum=0.5)
     Passthrough Float=&RandomRangeFloat(Minimum=0.4, Maximum=0.6)
    )
    Sprite Size=&MaskVector2dBySpawnGroup(
     Masked Vector2D=&RandomRangeVector2D(Minimum=<6 35>, Maximum=<12 63>)
     Passthrough Vector2D=&Vector2DFromFloat(Value=6.5)
     Spawn Group Mask=0
    )
    Sprite Size Mode=::NewEnumerator4
    UsePositionOffset=false
    Write Material Random=true
   ) {module=InitializeParticle001}
   &CylinderLocation(
    Cylinder Height=0
    Cylinder Radius=1
    MaskSpawnGroup=false
    Module.Enable Lathe Profile=false
    Module.Override Randomness=false
    Module.Surface Only=true
    Surface Only Band Thickness=1
    Use Endcaps In Surface Only Mode=false
   )
   &AddVelocity(Velocity=&MakeVector(
     X=0
     Y=0
     Z=&MaskFloatBySpawnGroup(Masked Float=&RandomRangeFloat(Minimum=50, Maximum=100), Passthrough Float=0)
    ))
   &AddVelocityFromPoint(
    Module.Use Falloff Distance=false
    Velocity Origin:Vector3f=&SimulationPosition()
    Velocity Strength=&MaskFloatBySpawnGroup(Masked Float=&Lerp_Float(A=250, B=450, Alpha=$System.HeightFactor), Passthrough Float=0)
   )
   &ApplyInitialForces()
   &DynamicMaterialParameters(
    Index 0 Param 1=&UniformRangedFloat(Minimum=0, Maximum=1)
    Index 0 Param 2=&UniformRangedFloat(Minimum=1, Maximum=2)
    Index 0 Param 3=1
    Index 0 Param 4=1
    Module.Param0WriteEnabled=true
    Module.Param1WriteEnabled=true
    Module.Param2WriteEnabled=true
    Module.Param3WriteEnabled=true
   )
   &SpriteFacingAndAlignment(Sprite Facing=<0 0 1>, WriteAlignment=false, WriteFacing=true)
  ]
  particleUpdateScript [
   &ParticleState()
   /SetVariables(Particles.SmokeGround_Camera=20) {isSetParameters=true}
   &ScaleSpriteSize(
    Scale Sprite Size Mode=::NewEnumerator0
    Uniform Scale Factor=&MaskFloatBySpawnGroup(
     Masked Float=1
     Passthrough Float=&FloatFromCurve(
      FloatCurve=DataInterface() {propertyValues {Curve {keys [
          {arriveTangent=35.6763, interpMode=RCIM_Cubic, leaveTangent=3.7835, tangentMode=RCTM_User, time=0, value=0}
          {interpMode=RCIM_Cubic, leaveTangent=0.6903, time=1.0011, value=1}
         ]}}}
      Scale Curve=50
     )
    )
   )
   /SetVariables(Particles.SmokeGround_Dynamic:Vector4f=&MakeVector4(
     X=&FloatFromCurve(FloatCurve=DataInterface() {propertyValues {Curve {keys [<at 0 0.5 RCIM_Cubic>,<at 1 2>]}}})
     Y=0
     Z=0
     W=0
    )) {isSetParameters=true}
   &Drag(Drag=1.5, Use Linear Drag=true, Use Rotational Drag=false)
   &SolveForcesAndVelocity(Clamp Velocity=false, Limit Acceleration=false)
   &ScaleColor(
    Scale Alpha=&FloatFromCurve(
     FloatCurve=DataInterface() {propertyValues {Curve {keys [<at 0 0 RCIM_Cubic>, {arriveTangent=0.002, interpMode=RCIM_Cubic, leaveTangent=0.002, time=0.5, value=1}, <at 1.0011 0.002 RCIM_Cubic>]}}}
     Scale Curve=0
    )
    Scale RGB=&MakeVectorFromLinearColorRGB(Color=&ColorFromCurve(ColorCurve=DataInterface() {propertyValues {
        AlphaCurve {keys [<at 0 0.5 RCIM_Cubic>,<at 1 0 RCIM_Cubic>]}
        BlueCurve {keys [<at 0 1.14 RCIM_Cubic>,<at 1 1 RCIM_Cubic>]}
        GreenCurve {keys [<at 0 1.842 RCIM_Cubic>,<at 1 1 RCIM_Cubic>]}
        RedCurve {keys [<at 0 3 RCIM_Cubic>,<at 1 1 RCIM_Cubic>]}
       }}))
    ScaleA=true
    ScaleRGB=true
   )
  ]
  rendererClasses [&NiagaraSpriteRendererProperties]
  renderers [
   {rendererClass=&NiagaraSpriteRendererProperties}
   {rendererClass=&NiagaraSpriteRendererProperties, rendererIndex=1}
  ]
 }
 Rim_Glow {
  emitterUpdateScript [
   &EmitterState()
   &SpawnRate(SpawnRate=4, Use Spawn Probability=false)
   &SpawnBurst_Instantaneous(Spawn Count=1, Use Spawn Probability=false) {enabled=false}
   /SetVariables(Emitter.FE_Toggle:NiagaraBool=&Boolean_NotOperation(Bool=&Boolean_OrOperation(Bool1=$User.bIsFrontEnd, Bool2=$User.bIsFrontEndPreview))) {isSetParameters=true}
  ]
  particleSpawnScript [
   &InitializeParticle(
    Color=<rgba 0.1348 0.0438 0.0135 0.5435>
    Mass Mode=::NewEnumerator1
    Sprite Size Mode=::NewEnumerator3
    Uniform Sprite Size=320
    UsePositionOffset=false
    Write Material Random=true
   ) {module=InitializeParticle001}
   &CameraOffset(Camera Offset Amount=3) {enabled=false}
   &SpriteFacingAndAlignment(Sprite Facing=<0 0 1>, WriteAlignment=false, WriteFacing=true)
  ]
  particleUpdateScript [
   &ParticleState()
   &ScaleSpriteSize(
    Uniform Curve Scale=0.9
    Uniform Curve Sprite Scale=DataInterface() {propertyValues {Curve {keys [{tangentMode=RCTM_User, time=0, value=0.5118}, {tangentMode=RCTM_User, time=1, value=1}]}}}
   )
   &FN_Light_v3(
    Color=<rgba 1 0.325 0.1 1>
    Create Color Attribute=true
    Create Exponent Attribute=true
    Create Radius Attribute=true
    Exponent=8
    Radius=120
    Scale Color=10
   )
   &self:Slope_Position_01(
    Apply Position=true
    Apply Rotation=true
    Height Adjustment/Offset=1
    Slope Position:NiagaraPosition=$Engine.Emitter.SimulationPosition
    Trace Length=120
    Trace Origin Height=85
    bIsFrontEnd:NiagaraBool=$User.bIsFrontEndPreview
   ) {module=Slope_Position_01}
   &ScaleColor(
    Scale Alpha=&FloatFromCurve(FloatCurve=DataInterface() {propertyValues {Curve {keys [<at 0 0>,<at 0.4237 0.934>,<at 1 0.3>]}}})
    ScaleA=true
    ScaleRGB=true
   )
  ]
  rendererClasses [&NiagaraSpriteRendererProperties, &NiagaraLightRendererProperties]
  renderers [
   {rendererClass=&NiagaraSpriteRendererProperties}
   {rendererClass=&NiagaraLightRendererProperties, rendererIndex=1}
  ]
 }
 Rim_Glow001 {
  emitterUpdateScript [
   &EmitterState()
   &SpawnRate(SpawnRate=4, Use Spawn Probability=false)
   &SpawnBurst_Instantaneous(Spawn Count=1, Use Spawn Probability=false) {enabled=false}
   /SetVariables(Emitter.FE:NiagaraBool=&Boolean_OrOperation(Bool1=$User.bIsFrontEnd, Bool2=$User.bIsFrontEndPreview)) {isSetParameters=true}
  ]
  particleSpawnScript [
   &InitializeParticle(
    Color=<rgba 0.1348 0.0438 0.0135 0.5435>
    Mass Mode=::NewEnumerator1
    Position Offset=<0 0 5>
    Sprite Size Mode=::NewEnumerator3
    Uniform Sprite Size=320
    UsePositionOffset=true
    Write Material Random=true
   ) {module=InitializeParticle001}
   &SpriteFacingAndAlignment(Sprite Facing=<0 0 1>, WriteAlignment=false, WriteFacing=true)
  ]
  particleUpdateScript [
   &ParticleState()
   &ScaleSpriteSize(
    Uniform Curve Scale=0.9
    Uniform Curve Sprite Scale=DataInterface() {propertyValues {Curve {keys [{tangentMode=RCTM_User, time=0, value=0.5118}, {tangentMode=RCTM_User, time=1, value=1}]}}}
   )
   &FN_Light_v3(
    Color=<rgba 1 0.325 0.1 1>
    Create Color Attribute=true
    Create Exponent Attribute=true
    Create Radius Attribute=true
    Exponent=8
    Radius=120
    Scale Color=10
   )
   &ScaleColor(
    Scale Alpha=&FloatFromCurve(FloatCurve=DataInterface() {propertyValues {Curve {keys [<at 0 0>,<at 0.4237 0.934>,<at 1 0.3>]}}})
    ScaleA=true
    ScaleRGB=true
   )
  ]
  rendererClasses [&NiagaraSpriteRendererProperties]
  renderers [{rendererClass=&NiagaraSpriteRendererProperties}]
 }
 smoke_pulse {
  emitterUpdateScript [
   &EmitterState(Life Cycle Mode=::NewEnumerator1, Loop Duration=0.2, UseLoopDelay=false)
   &SpawnBurst_Instantaneous(
    Spawn Count=&MakeIntFromFloat(Float=&Lerp_Float(A=10, B=15, Alpha=$System.HeightFactor))
    Use Spawn Probability=false
   )
  ]
  particleSpawnScript [
   &InitializeParticle(
    Color=&MakeLinearColorFromVectorAndFloat(
     "Vector (RGB)"=&Multiply_Vector(
      A=&MakeCustomVectorFromBool(
       Bool=$User.bIsFrontEndPreview
       True Vector=&MakeVectorFromLinearColorRGB(Color=<rgba 0.4 0.4 0.4 1>)
       False Vector=&MakeVectorFromLinearColorRGB(Color=<rgba 0.13 0.13 0.13 1>)
      )
      B=&VectorFromFloat(Value=1)
     )
     "Float (Alpha)"=&Multiply_Float(
      A=&MakeCustomFloatFromBool(
       Bool=$User.bIsFrontEndPreview
       True Float=0.03
       False Float=&FN_Select_Float_from_Time_Of_Day(Use Manual Time of Day Thresholds=true, Morning=0.12, Day=0.12, Evening=0.09, Night=0.17, Evening Threshold=20)
      )
      B=&Lerp_Float(A=0.6, B=1.2, Alpha=$System.HeightFactor)
     )
    )
    Lifetime=&RandomRangeFloat(Minimum=0.6, Maximum=0.7)
    Sprite UV Mode=::NewEnumerator3
    UsePositionOffset=false
    Write Material Random=true
   )
   &CylinderLocation(
    Cylinder Height=0
    Cylinder Origin:Vector3f=$System.SmokeCenterPositionWS
    Cylinder Radius=30
    MaskSpawnGroup=false
    Module.Enable Lathe Profile=false
    Module.Override Randomness=false
    Module.Surface Only=true
    Surface Only Band Thickness=0
   )
   &AddVelocityFromPoint(
    Module.Use Falloff Distance=false
    Velocity Origin:Vector3f=$System.SmokeCenterPositionWS
    Velocity Strength=&Multiply_Float(
     A=&Lerp_Float(A=320, B=420, Alpha=$Particles.MaterialRandom)
     B=&Lerp_Float(A=0.8, B=1.5, Alpha=$System.HeightFactor)
    )
   )
   /SetVariables(Particles.SpriteSize=&Vector2DFromFloat(Value=&MaskFloatBySpawnGroup(
      Masked Float=&RandomRangeFloat(Minimum=90, Maximum=120)
      Passthrough Float=&RandomRangeFloat(Minimum=110, Maximum=140)
     ))) {isSetParameters=true}
  ]
  particleUpdateScript [
   &ParticleState()
   &LinearForce(
    Coordinate Space:ENiagaraCoordinateSpace=::World
    Force=&Lerp_Vector(A=<0 0 0>, B=<0 0 110>, Alpha=$System.HeightFactor)
   )
   &Drag(Drag=7, Use Linear Drag=true, Use Rotational Drag=false)
   &SolveForcesAndVelocity(Clamp Velocity=false, Limit Acceleration=false, Speed Limit=1)
   &ScaleColor(
    Scale Alpha=&FloatFromCurve(
     FloatCurve=DataInterface() {propertyValues {Curve {keys [
         <at 0 0 RCIM_Cubic>
         {arriveTangent=-1.9555, interpMode=RCIM_Cubic, leaveTangent=-1.9555, tangentMode=RCTM_User, time=0.2374, value=1}
         <at 1 0>
        ]}}}
     Scale Curve=1.5
    )
    ScaleA=true
    ScaleRGB=true
   )
   &ScaleSpriteSize(Scale Factor=&Vector2DFromFloat(Value=&FloatFromCurve(FloatCurve=DataInterface() {propertyValues {Curve {keys [<at 0 0.5 RCIM_Cubic>, {arriveTangent=0.5532, interpMode=RCIM_Cubic, leaveTangent=-1.4995, tangentMode=RCTM_User, time=1, value=1}]}}})))
   &DynamicMaterialParameters(
    Index 0 Param 1=&FloatFromCurve(FloatCurve=DataInterface() {propertyValues {Curve {keys [<at 0 1>,<at 1 0>]}}}, Scale Curve=0.8)
    Index 0 Param 2=&FN_Select_Float_from_Time_Of_Day(Use Manual Time of Day Thresholds=true, Morning=1, Day=1, Evening=1, Night=0.2, Evening Threshold=20)
    Module.Param0WriteEnabled=true
    Module.Param1WriteEnabled=true
    Module.Param2WriteEnabled=true
    Module.Param3WriteEnabled=true
   )
   &SubUVAnimation(SubUV Loop Count=0)
   &self:CustomCameraOffset(
    Curve=DataInterface() {propertyValues {Curve {keys [
        {tangentMode=RCTM_User, time=0, value=0}
        {interpMode=RCIM_Cubic, leaveTangent=6.6844, tangentMode=RCTM_Break, time=0.4, value=0}
        {tangentMode=RCTM_User, time=0.8, value=1}
       ]}}}
    OffsetValue:Vector3f=<0 0 8>
   ) {module=CustomCameraOffset}
   &self:CameraDistanceFade(
    MaxDistance=2000
    New Curve for Floats=DataInterface() {propertyValues {Curve {keys [
        <at 0 1>
        {interpMode=RCIM_Cubic, leaveTangent=0.2164, tangentMode=RCTM_Break, time=0.1, value=1}
        {arriveTangent=-0.194, interpMode=RCIM_Cubic, tangentMode=RCTM_Break, time=0.5, value=0.4}
       ]}}}
   ) {module=CameraDistanceFade}
   &ScaleColor(Scale Alpha=$Particles.CameraDistanceAlpha, ScaleA=true, ScaleRGB=true) {module=ScaleColor001}
  ]
  rendererClasses [&NiagaraSpriteRendererProperties]
  renderers [{rendererClass=&NiagaraSpriteRendererProperties}]
 }
}
summary {
 systemName=NS_Emote_ScoopPluck_Dust
 emitters [
  {emitterName=GroundRocks, rendererClasses [&NiagaraSpriteRendererProperties]}
  {emitterName=smoke_pulse, rendererClasses [&NiagaraSpriteRendererProperties]}
  {emitterName=GroundSmoke, rendererClasses [&NiagaraSpriteRendererProperties]}
  {emitterName=Rim_Glow, rendererClasses [&NiagaraSpriteRendererProperties, &NiagaraLightRendererProperties]}
  {emitterName=Rim_Glow001, rendererClasses [&NiagaraSpriteRendererProperties]}
 ]
 userVariables [
  User.bIsFrontEnd {type=NiagaraBool}
  User.bIsFrontEndPreview {type=NiagaraBool}
  User.bIsInWater {type=NiagaraBool}
  User.Flame_Size {type=NiagaraFloat, defaultValue=1, description=""}
  User.InitDelay {type=NiagaraFloat, defaultValue=0, description=""}
  User.OwnerLocalVelocity {type=Vector3f, description="", defaultValue {struct=Vector3f, value=<0 0 0>}}
  User.PlayerSpeed {type=NiagaraFloat, defaultValue=0, description=""}
  User.SkeletalMesh {
   type=NiagaraDataInterfaceSkeletalMesh
   defaultValue {
    struct=NiagaraExt_VariableValue_DataInterface
    value {dataInterface=&self:NiagaraDataInterfaceSkeletalMesh_0}
   }
  }
 ]
}
