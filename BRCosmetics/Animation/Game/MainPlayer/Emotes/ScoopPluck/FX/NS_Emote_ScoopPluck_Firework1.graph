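# Niagara effect for the ScoopPluck emote's fireworks, launched from the gadget's left and right rocket sockets.
# Socket-attached glow and streaks mark the launch, while noise-jittered flares in random gold, pink or blue climb, shedding ribbon smoke and embers.
# Flares write death position and color to system arrays that drive burst flashes and sparks.
# Spark count scales by quality level; burst brightness follows a time-of-day multiplier.
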NiagaraSystem /BRCosmetics/Animation/Game/MainPlayer/Emotes/ScoopPluck/FX/NS_Emote_ScoopPluck_Firework1
dependencies {
 usedDataInterfaces [
  &NiagaraDataInterfaceParticleRead
  &NiagaraDataInterfaceCamera
  &NiagaraDataInterfaceArrayColor
  &NiagaraDataInterfaceSkeletalMesh
  &NiagaraDataInterfaceArrayFloat
  &NiagaraDataInterfaceArrayFloat3
  &NiagaraDataInterfaceVector2DCurve
  &NiagaraDataInterfaceArrayInt32
  &NiagaraDataInterfaceArrayPosition
 ]
 usedDynamicInputs [
  &MakeVector4_FromFloat
  &MakeVector2D
  &FNTimeOfDayScalar
 ]
 usedModules [
  &self:SystemSpawnScript.SystemScriptSource.SystemScriptGraph.NiagaraNodeAssignment_7.SetVariables_#1
  &SystemState
  &self:SystemSpawnScript.SystemScriptSource.SystemScriptGraph.NiagaraNodeAssignment_2.SetVariables_#2
  &self:ReadDeath_02
  &self:SystemSpawnScript.SystemScriptSource.SystemScriptGraph.NiagaraNodeAssignment_8.SetVariables_#3
 ]
}
emitters {
 FireworkFlare_Parent001 {
  emitterUpdateScript [
   &EmitterState()
   &SpawnBurst_Instantaneous(Use Spawn Probability=false, UseLoopCountLimit=false)
   &SpawnBurst_Instantaneous(
    Spawn Count=&RandomRangeInt(Minimum=1, Maximum=1)
    Spawn Time=0.05
    Use Spawn Probability=false
    UseLoopCountLimit=false
   ) {module=SpawnBurst_Instantaneous001}
   &SpawnBurst_Instantaneous(Spawn Time=0.1, Use Spawn Probability=false, UseLoopCountLimit=false) {module=SpawnBurst_Instantaneous002}
   &SpawnBurst_Instantaneous(
    Spawn Count=&RandomRangeInt(Minimum=1, Maximum=2)
    Spawn Time=0.15
    Use Spawn Probability=false
    UseLoopCountLimit=false
   ) {module=SpawnBurst_Instantaneous003}
   &SpawnBurst_Instantaneous(Spawn Time=0.2, Use Spawn Probability=false, UseLoopCountLimit=false) {module=SpawnBurst_Instantaneous004}
  ]
  particleSpawnScript [
   &FN_SetSpawnGroup(Number of Spawn Groups=2)
   &InitializeParticle(
    Color=&SelectLinearColorFromArray(
     Array Sampling Mode=::NewEnumerator0
     Direct Array Index=?
     Color Selection Array=DataInterface() {propertyValues {ColorData [<rgba 5 0.9006 0 1>, <rgba 5 0.7061 1.2761 1>, <rgba 0.3319 1.3976 5 1>]}}
     Random Seed=0
    )
    Lifetime Max=0.6
    Lifetime Min=0.2
    Lifetime Mode=::NewEnumerator1
    Sprite Rotation Mode=::NewEnumerator2
    Sprite Size Mode=::NewEnumerator1
    Uniform Sprite Size Max=20
    Uniform Sprite Size Min=15
    UsePositionOffset=false
    Write Material Random=true
   )
   &SkeletalMeshLocation(
    MaskSpawnGroup=false
    Mesh Sampling Type=::NewEnumerator5
    OverrideRandomSeed=false
    Skeletal Mesh=DataInterface() {propertyValues {
      DefaultMesh=null
      ExcludeBoneName=root
      OptionalComponentClass=null
      PreviewMesh=/BRCosmetics/Animation/Props/Gadgets_ScoopPluck/Meshes/ScoopPluck_Gadget
      SkinningMode=SkinOnTheFly
      SoftSourceActor=null
      SourceMode=Default
      UvSetIndex=0
      WholeMeshLOD=-1
      bExcludeBone=false
      bReadDeformedGeometry=true
      bRequireCurrentFrameData=true
      ComponentTags []
      FilteredBones []
      FilteredSockets [JNT_R_ROCKET_SOCKET, JNT_L_ROCKET_SOCKET]
      MeshUserParameter {parameter {name=None, typeDefHandle {registeredTypeIndex=1}, varData []}}
      SamplingRegions []
     }}
   )
   &ShapeLocation(
    Box / Plane Mode:ENiagara_BoxPlaneMode=::NewEnumerator1
    MaskSpawnGroup=false
    Offset=&MakeVector(
     X=&MaskFloatBySpawnGroup(
      Masked Float=&FloatFromCurve(FloatCurve=DataInterface() {propertyValues {Curve {keys [<at 0 -5>,<at 1 30>]}}}, CurveIndex=$Emitter.NormalizedLoopAge)
      Passthrough Float=&FloatFromCurve(FloatCurve=DataInterface() {propertyValues {Curve {keys [<at 0 5>,<at 1 -30>]}}}, CurveIndex=$Emitter.NormalizedLoopAge)
     )
     Y=0
     Z=0
    )
    Offset Mode=::NewEnumerator0
    Override Seed=false
    Plane Size:Vector2f=<0 0>
    Shape Primitive=::NewEnumerator2
    UseOwnerScale=true
   )
   &AddVelocity(
    Module.Use Falloff Distance=false
    Origin Offset=&MaskVectorBySpawnGroup(Masked Vector=<-20 0 -100>, Passthrough Vector=<20 0 -100>)
    Override Seed=false
    Velocity Mode=::NewEnumerator1
    Velocity Speed=&RandomRangeFloat(Minimum=1500, Maximum=1850)
   ) {module=AddVelocity002}
  ]
  particleUpdateScript [
   /Niagara/Noise/Waveform(
    EnableExponent1=false
    Global Phase:NiagaraFloat=$Emitter.NormalizedLoopAge
    GlobalClampMax=false
    GlobalClampMin=false
    UseInput1=true
    UsePhase1=true
    "Waveform [1]":ENiagara_Waveforms=::NewEnumerator12
    "[1] Frequency"=2
    "[1] Phase":NiagaraFloat=$Emitter.NormalizedLoopAge
    "[1] Phase Scale"=5
   ) {module=Waveform001}
   &GravityForce(
    Coordinate Space=::Local
    Gravity=&MakeVector(
     X=0
     Y=&MaskFloatBySpawnGroup(
      Masked Float=&FloatFromCurve(
       FloatCurve=DataInterface() {propertyValues {Curve {keys [
           {arriveTangent=0.7011, interpMode=RCIM_Cubic, leaveTangent=0.7011, tangentMode=RCTM_User, time=0, value=0}
           {arriveTangent=1.3889, arriveTangentWeight=0.3333, interpMode=RCIM_Cubic, leaveTangent=1.3889, tangentMode=RCTM_User, time=1, value=1}
          ]}}}
       CurveIndex=$Emitter.NormalizedLoopAge
       Scale Curve=-2150
      )
      Passthrough Float=&FloatFromCurve(
       FloatCurve=DataInterface() {propertyValues {Curve {keys [
           {arriveTangent=0.0446, interpMode=RCIM_Cubic, leaveTangent=0.0446, tangentMode=RCTM_User, time=0, value=0}
           {arriveTangent=1.9596, interpMode=RCIM_Cubic, leaveTangent=1.9596, tangentMode=RCTM_User, time=1, value=1}
          ]}}}
       CurveIndex=$Emitter.NormalizedLoopAge
       Scale Curve=-2150
      )
     )
     Z=&RandomRangeFloat(Minimum=-980, Maximum=-1100)
    )
   )
   &Drag(Drag=&RandomRangeFloat(Minimum=5, Maximum=8), Use Linear Drag=true, Use Rotational Drag=false)
   &VectorNoiseForce(
    Force Amount=&Lerp_Float(A=-2500, B=2500, Alpha=$StackContext.Waveform001.Waveform)
    Module.Use Random Noise Time Interval=false
   )
   &SkeletalMeshLocation(
    Bone Sampling Mode=::NewEnumerator4
    MaskSpawnGroup=false
    Orientation Sampling=::NewEnumerator1
    Position Sampling=::NewEnumerator2
    Skeletal Mesh=DataInterface() {propertyValues {
      DefaultMesh=null
      ExcludeBoneName=root
      OptionalComponentClass=null
      PreviewMesh=&ScoopPluck_Gadget
      SkinningMode=SkinOnTheFly
      SoftSourceActor=null
      SourceMode=Default
      UvSetIndex=0
      WholeMeshLOD=-1
      bExcludeBone=false
      bReadDeformedGeometry=true
      bRequireCurrentFrameData=true
      ComponentTags []
      FilteredBones [Root]
      FilteredSockets []
      MeshUserParameter {parameter {name=None, typeDefHandle {registeredTypeIndex=1}, varData []}}
      SamplingRegions []
     }}
   ) {module=SkeletalMeshLocation001}
   &InheritSourceMovement(
    ApplyForce=false
    ApplyPosition=true
    ApplyPositionFalloff=false
    Source Velocity:Vector3f=$Particles.SkeletalMeshLocation001.SampledVelocity
   )
   &SolveForcesAndVelocity(Clamp Velocity=false, Limit Acceleration=false)
   &ParticleState()
   &ScaleSpriteSize(Scale Factor=&Vector2DFromFloat(Value=&FloatFromCurve(
      FloatCurve=DataInterface() {propertyValues {Curve {keys [
          {arriveTangent=6.6282, interpMode=RCIM_Cubic, leaveTangent=6.6282, tangentMode=RCTM_User, time=0, value=0}
          <at 0.1 0.4626>
          {arriveTangent=4.0254, interpMode=RCIM_Cubic, leaveTangent=4.0254, time=0.1078, value=0.2796}
          <at 0.15 0.6638>
          <at 0.2 0.2796>
          <at 0.25 0.5723>
          <at 0.3 0.2887>
          <at 0.35 0.8011>
          <at 0.4 0.2887>
          {arriveTangent=-2.7618, leaveTangent=-2.7618, time=0.45, value=1}
          {arriveTangent=1, leaveTangent=1, time=0.55, value=0.3}
          {arriveTangent=-0.75, leaveTangent=-0.75, time=0.65, value=1.15}
          <at 0.75 0.5>
          {arriveTangent=-6.1253, interpMode=RCIM_Cubic, leaveTangent=-6.1253, tangentMode=RCTM_User, time=0.8, value=1.4}
          <at 1 0>
         ]}}}
      Scale Curve=3
     ))) {module=ScaleSpriteSize002}
   &DynamicMaterialParameters(
    Index 0 Param 1=0.8
    Index 0 Param 2=1
    Index 0 Param 3=1
    Index 1 Param 1=150
    Index 1 Param 4=1
    Module.Index0Float0Write=true
    Module.Index0Float1Write=true
    Module.Index0Float2Write=true
    Module.Index0Float3Write=true
    Module.Index1Float0Write=true
    Module.Index1Float1Write=true
    Module.Index1Float2Write=true
    Module.Index1Float3Write=true
    Write Parameter Index 1=::NewEnumerator1
   )
   &ScaleColor(
    Scale Alpha=&FloatFromCurve(
     FloatCurve=DataInterface() {propertyValues {Curve {keys [
         {arriveTangent=-4.6344, interpMode=RCIM_Cubic, leaveTangent=-4.6344, tangentMode=RCTM_User, time=0.5, value=1}
         {arriveTangent=-0.2005, tangentMode=RCTM_User, time=1, value=0}
        ]}}}
     Scale Curve=&RandomRangeFloat(Minimum=0.75, Maximum=&FN_FrontEndFloatModifier(Gameplay=1.25, Front End=1))
    )
    Scale RGB=&VectorFromFloat(Value=&FloatFromCurve(
      FloatCurve=DataInterface() {propertyValues {Curve {keys [
          {arriveTangent=-1.6398, interpMode=RCIM_Cubic, leaveTangent=-1.6398, tangentMode=RCTM_User, time=0, value=2}
          {arriveTangent=0.009, tangentMode=RCTM_User, time=1, value=1}
         ]}}}
      Scale Curve=&RandomRangeFloat(Minimum=0.75, Maximum=&FN_FrontEndFloatModifier(Gameplay=1.25, Front End=1))
     ))
    ScaleA=false
    ScaleRGB=true
   )
   &FN_Light_Attributes(
    Color=$Particles.Color
    Create Color Attribute=true
    Create Exponent Attribute=true
    Create Radius Attribute=true
    Exponent=4
    Radius=120
    Scale Color=&FloatFromCurve(
     FloatCurve=DataInterface() {propertyValues {Curve {keys [
         {interpMode=RCIM_Cubic, tangentMode=RCTM_User, time=0, value=1}
         {arriveTangent=-2.6287, interpMode=RCIM_Cubic, leaveTangent=-2.6287, tangentMode=RCTM_User, time=1, value=0}
        ]}}}
     Scale Curve=0.25
    )
   )
   /SetVariables(
    Particles.CrossColor:LinearColor=&Multiply_LinearColorByFloat(Linear Color=$Particles.Color, Float=0.5)
    Particles.CrossRotation:NiagaraFloat=&RandomRangeFloat(Minimum=0, Maximum=60)
    Particles.CrossSize:Vector2f=&Vector2DFromFloat(Value=&FloatFromCurve(
      FloatCurve=DataInterface() {propertyValues {Curve {keys [
          {arriveTangent=8.9932, interpMode=RCIM_Cubic, leaveTangent=8.9932, tangentMode=RCTM_User, time=0, value=0}
          {arriveTangent=0.1411, interpMode=RCIM_Cubic, leaveTangent=0.1411, tangentMode=RCTM_User, time=0.25, value=1}
          {arriveTangent=-1.3601, interpMode=RCIM_Cubic, leaveTangent=-1.3601, tangentMode=RCTM_User, time=1.0067, value=0.5}
         ]}}}
      Scale Curve=&RandomRangeFloat(Minimum=70, Maximum=90)
     ))
   ) {isSetParameters=true}
   &SpriteRotationRate(Rotation Rate=&FloatFromCurve(
     FloatCurve=DataInterface() {propertyValues {Curve {keys [<at 0 0>,<at 0.1 0>,<at 0.3 1>]}}}
     Scale Curve=&SelectFloatFromArray(
      Array Sampling Mode=::NewEnumerator0
      Direct Array Index=?
      Float Selection Array=DataInterface() {propertyValues {FloatData [300,-300]}}
      Random Seed=0
     )
    ))
   &self:WriteDeath_02(Color Array:NiagaraDataInterfaceArrayColor=$System.ColorArray, Position Array:NiagaraDataInterfaceArrayPosition=$System.DeathArray) {module=WriteDeath_02}
  ]
  rendererClasses [&NiagaraSpriteRendererProperties, &NiagaraLightRendererProperties]
  renderers [
   {rendererClass=&NiagaraSpriteRendererProperties}
   {rendererClass=&NiagaraLightRendererProperties, rendererIndex=1}
   {rendererClass=&NiagaraSpriteRendererProperties, rendererIndex=2}
  ]
 }
 Flare_With_Streak002 {
  emitterUpdateScript [
   &EmitterState()
   &SpawnPerFrame(Spawn:NiagaraBool=$System.DeathSpawn, Use Spawn Probability=false)
   &SpawnBurst_Instantaneous(Use Spawn Probability=false, UseLoopCountLimit=false) {enabled=false}
  ]
  particleSpawnScript [
   &InitializeParticle(
    Color=$System.DeathColor
    Lifetime=0.2
    Position=$System.DeathPosition
    Position Mode=::NewEnumerator1
    Sprite Rotation Mode=::NewEnumerator2
    Sprite Size Mode=::NewEnumerator3
    Uniform Sprite Size=200
    UsePositionOffset=false
    Write Material Random=true
   )
   &DynamicMaterialParameters(
    Index 0 Param 1=&RandomRangeFloat(Minimum=0.4, Maximum=0.6)
    Index 0 Param 2=1
    Index 0 Param 3=1
    Index 0 Param 4=1
    Module.Param0WriteEnabled=true
    Module.Param1WriteEnabled=true
    Module.Param2WriteEnabled=true
    Module.Param3WriteEnabled=true
   )
   /SetVariables() {isSetParameters=true}
  ]
  particleUpdateScript [
   &ParticleState()
   &SolveForcesAndVelocity(Clamp Velocity=false, Limit Acceleration=false)
   &ScaleColor(
    Scale Alpha=&Multiply_Float(
     A=0.15
     B=&FloatFromCurve(
      FloatCurve=DataInterface() {propertyValues {Curve {keys [<at 0 0>,<at 0.2 1>,<at 0.8 1>,<at 1 0>]}}}
      Scale Curve=$System.TimeOfDayBrightnessMultiplier
     )
    )
    ScaleA=true
    ScaleRGB=true
   )
   &ScaleSpriteSize(Scale Factor=&Vector2DFromFloat(Value=&FloatFromCurve(
      FloatCurve=DataInterface() {propertyValues {Curve {keys [<at 0 1.5 RCIM_Cubic>,<at 1 0>]}}}
      Scale Curve=&FN_FrontEndFloatModifier(Gameplay=1, Front End=0.7)
     )))
   &CameraOffset(Camera Offset Amount=50)
   /SetVariables(Particles.Flash size:Vector2f=&Multiply_Vector2D_ByFloat(
     Float=&Multiply_Float(
      A=30
      B=&FloatFromCurve(
       FloatCurve=DataInterface() {propertyValues {Curve {keys [
           {arriveTangent=4.1785, interpMode=RCIM_Cubic, leaveTangent=4.1785, tangentMode=RCTM_User, time=0, value=0}
           {arriveTangent=0.9995, interpMode=RCIM_Cubic, leaveTangent=0.9995, tangentMode=RCTM_User, time=0.1686, value=0.5887}
           {arriveTangent=-0.0405, interpMode=RCIM_Cubic, leaveTangent=-0.0405, tangentMode=RCTM_User, time=0.8675, value=1}
          ]}}}
       Scale Curve=&FN_FrontEndFloatModifier(Gameplay=0.9, Front End=0.7)
      )
     )
     Vector2D=<1.2 1>
    )) {isSetParameters=true}
   /SetVariables(Particles.Flash color:LinearColor=&MakeLinearColorFromVectorAndFloat(
     "Vector (RGB)"=&MakeVectorFromLinearColorRGB(Color=$System.DeathColor)
     "Float (Alpha)"=&Multiply_Float(
      A=&Multiply_Float(
       A=1
       B=&FloatFromCurve(
        FloatCurve=DataInterface() {propertyValues {Curve {keys [<at 0 0>,<at 0.124 1>,<at 0.496 1>,<at 0.62 0>]}}}
        Scale Curve=$System.TimeOfDayBrightnessMultiplier
       )
      )
      B=1
     )
    )) {enabled=false, isSetParameters=true}
   &FN_Light_Attributes(
    Color=$Particles.Initial.Color
    Create Color Attribute=true
    Create Exponent Attribute=true
    Create Radius Attribute=true
    Exponent=4
    Radius=150
    Scale Color=&FloatFromCurve(FloatCurve=DataInterface() {propertyValues {Curve {
        defaultValue=1
        keys [
         {tangentMode=RCTM_User, time=0, value=1}
         {arriveTangent=-0.0405, interpMode=RCIM_Cubic, leaveTangent=-0.0405, tangentMode=RCTM_User, time=0.7009, value=1}
         {arriveTangent=-8.0089, interpMode=RCIM_Cubic, leaveTangent=-8.0089, tangentMode=RCTM_User, time=1, value=0}
        ]
       }}})
   )
  ]
  rendererClasses [&NiagaraSpriteRendererProperties, &NiagaraLightRendererProperties]
  renderers [
   {rendererClass=&NiagaraSpriteRendererProperties}
   {rendererClass=&NiagaraSpriteRendererProperties, rendererIndex=1}
   {rendererClass=&NiagaraLightRendererProperties, rendererIndex=2}
  ]
 }
 Glow {
  emitterUpdateScript [
   &EmitterState()
   &SpawnBurst_Instantaneous(Spawn Count=2, Use Spawn Probability=false, UseLoopCountLimit=false)
   &SpawnBurst_Instantaneous(Spawn Count=2, Spawn Time=0.08, Use Spawn Probability=false, UseLoopCountLimit=false) {module=SpawnBurst_Instantaneous001}
   &SpawnBurst_Instantaneous(Spawn Count=2, Spawn Time=0.16, Use Spawn Probability=false, UseLoopCountLimit=false) {module=SpawnBurst_Instantaneous002}
   &SpawnBurst_Instantaneous(Spawn Count=2, Spawn Time=0.0333, Use Spawn Probability=false, UseLoopCountLimit=false) {module=SpawnBurst_Instantaneous003}
  ]
  particleSpawnScript [
   &InitializeParticle(
    Color=<rgba 15 6.7784 1.6264 0.2>
    Lifetime=0.08
    Mesh Scale=<0.8 0.65 0.65>
    Mesh Scale Mode=::NewEnumerator4
    Position Offset=<0 0 18>
    Sprite Rotation Mode=::NewEnumerator0
    Sprite Size Mode=::NewEnumerator3
    Uniform Sprite Size=100
    UsePositionOffset=true
    Write Material Random=true
   )
   &PartitionParticles()
   &SkeletalMeshLocation(
    Bone / Socket Index=$Particles.PartitionParticles.Partition
    MaskSpawnGroup=false
    Mesh Sampling Type=::NewEnumerator5
    Sampled Position Offset=&SelectVectorFromArray(
     Direct Array Index=$Particles.PartitionParticles.Partition
     Vector Selection Array=DataInterface() {propertyValues {FloatData [<0 0 1>, <0 0 -1>]}}
    )
    Skeletal Mesh=DataInterface() {propertyValues {
      DefaultMesh=null
      ExcludeBoneName=root
      OptionalComponentClass=null
      PreviewMesh=&ScoopPluck_Gadget
      SkinningMode=SkinOnTheFly
      SoftSourceActor=null
      SourceMode=Default
      UvSetIndex=0
      WholeMeshLOD=-1
      bExcludeBone=false
      bReadDeformedGeometry=true
      bRequireCurrentFrameData=true
      ComponentTags []
      FilteredBones []
      FilteredSockets [JNT_R_ROCKET_SOCKET, JNT_L_ROCKET_SOCKET]
      MeshUserParameter {parameter {name=None, typeDefHandle {registeredTypeIndex=1}, varData []}}
      SamplingRegions []
     }}
    Socket Sampling Mode=::NewEnumerator5
   )
   &InitialMeshOrientation(
    Mesh Orientation Mode=::NewEnumerator6
    Rotation=&SelectVectorFromArray(
     Direct Array Index=$Particles.PartitionParticles.Partition
     Vector Selection Array=DataInterface() {propertyValues {FloatData [<0.25 0 -0.1>, <0.25 0 0.1>]}}
    )
    Use Rotation Vector=true
   )
   /SetVariables(Particles.FlashSize:Vector2f=&Multiply_Vector2D(A=$Particles.SpriteSize, B=<2 0.5>)) {isSetParameters=true}
   &CameraOffset()
   /SetVariables(Particles.LensFlareColor:LinearColor=&MultiplyLinearColor(Linear Color=$Particles.Color, Scale Factor=<0.3 0.3 0.3 0.3>)) {isSetParameters=true}
   /SetVariables(Particles.FlareCameraOffset=20) {isSetParameters=true}
  ]
  particleUpdateScript [
   &ParticleState()
   &SkeletalMeshLocation(
    Bone / Socket Index=$Particles.PartitionParticles.Partition
    MaskSpawnGroup=false
    Mesh Sampling Type=::NewEnumerator5
    Orientation Sampling=::NewEnumerator1
    Sampled Position Offset=&SelectVectorFromArray(
     Direct Array Index=$Particles.PartitionParticles.Partition
     Vector Selection Array=DataInterface() {propertyValues {FloatData [<0 0 1>, <0 0 -1>]}}
    )
    Skeletal Mesh=DataInterface() {propertyValues {
      DefaultMesh=null
      ExcludeBoneName=root
      OptionalComponentClass=null
      PreviewMesh=&ScoopPluck_Gadget
      SkinningMode=SkinOnTheFly
      SoftSourceActor=null
      SourceMode=Default
      UvSetIndex=0
      WholeMeshLOD=-1
      bExcludeBone=false
      bReadDeformedGeometry=true
      bRequireCurrentFrameData=true
      ComponentTags []
      FilteredBones []
      FilteredSockets [JNT_R_ROCKET_SOCKET, JNT_L_ROCKET_SOCKET]
      MeshUserParameter {parameter {name=None, typeDefHandle {registeredTypeIndex=1}, varData []}}
      SamplingRegions []
     }}
    Socket Sampling Mode=::NewEnumerator5
   ) {module=SkeletalMeshLocation001}
   /SetVariables(Particles.NewPos:NiagaraPosition=&AddVectorToPosition(
     Position=$Particles.Position
     Vector=&SelectVectorFromArray(
      Direct Array Index=$Particles.PartitionParticles.Partition
      Vector Selection Array=DataInterface() {propertyValues {FloatData [<3 0 7>, <-3 0 7>]}}
     )
    )) {isSetParameters=true}
   /SetVariables(Particles.FlarePos:NiagaraPosition=&AddVectorToPosition(
     Position=$Particles.Position
     Vector=&SelectVectorFromArray(
      Direct Array Index=$Particles.PartitionParticles.Partition
      Vector Selection Array=DataInterface() {propertyValues {FloatData [<3 0 5>, <-3 0 5>]}}
     )
    )) {isSetParameters=true}
   &ScaleColor(
    Scale Alpha=&FloatFromCurve(FloatCurve=DataInterface() {propertyValues {Curve {keys [
         <at 0 0>
         {tangentMode=RCTM_User, time=0.2, value=1}
         {arriveTangent=-0.0405, interpMode=RCIM_Cubic, leaveTangent=-0.0405, tangentMode=RCTM_User, time=0.9, value=1}
         {arriveTangent=-8.0089, interpMode=RCIM_Cubic, leaveTangent=-8.0089, tangentMode=RCTM_User, time=1, value=0}
        ]}}})
    ScaleA=false
    ScaleRGB=true
   )
   &ScaleSpriteSize(Uniform Curve Sprite Scale=DataInterface() {propertyValues {Curve {keys [
        <at 0 0>
        <at 0.2633 1>
        {arriveTangent=-0.0405, interpMode=RCIM_Cubic, leaveTangent=-0.0405, tangentMode=RCTM_User, time=0.9, value=1.5}
        {arriveTangent=-8.0089, interpMode=RCIM_Cubic, leaveTangent=-8.0089, tangentMode=RCTM_User, time=1, value=0}
       ]}}})
   &DynamicMaterialParameters(
    Index 0 Param 1=0.1
    Index 0 Param 2=1.7
    Module.Index0Float0Write=true
    Module.Index0Float1Write=true
    Module.Index0Float2Write=true
    Module.Index0Float3Write=true
   )
   &FN_Light_Attributes(
    Color=&ColorFromCurve(ColorCurve=DataInterface() {propertyValues {
       AlphaCurve {keys [<at 0 0>,<at 0.2 1 RCIM_Cubic>,<at 0.75 1>,<at 1 0 RCIM_Cubic>]}
       BlueCurve {keys [<at 0 0.1406 RCIM_Cubic>]}
       GreenCurve {keys [<at 0 0.5823 RCIM_Cubic>]}
       RedCurve {keys [<at 0 1 RCIM_Cubic>]}
      }})
    Create Color Attribute=true
    Create Exponent Attribute=true
    Create Radius Attribute=true
    Radius=200
    Scale Color=9
   )
   &DynamicMaterialParameters(
    Index 0 Param 1=&FloatFromCurve(FloatCurve=DataInterface() {propertyValues {Curve {keys [<at 0 0>,<at 1 0.5>]}}})
    Module.Index0Float0Write=true
    Module.Index0Float1Write=true
    Module.Index0Float2Write=true
    Module.Index0Float3Write=true
   ) {module=DynamicMaterialParameters001}
   &ScaleMeshSize(Scale Factor=&Multiply_VectorByFloat(
     Vector=&Multiply_VectorByFloat(
      Vector=&MakeVector(
       X=&FloatFromCurve(FloatCurve=DataInterface() {propertyValues {Curve {keys [<at 0 1>,<at 1 2>]}}})
       Y=1
       Z=1
      )
      Float=2
     )
     Float=&FloatFromCurve(
      FloatCurve=DataInterface() {propertyValues {Curve {keys [<at 0 0.5>,<at 1 1>]}}}
      Scale Curve=&RandomRangeFloat(Minimum=0.2, Maximum=0.2)
     )
    ))
  ]
  rendererClasses [
   &NiagaraSpriteRendererProperties
   &NiagaraLightRendererProperties
   &NiagaraMeshRendererProperties
  ]
  renderers [
   {rendererClass=&NiagaraSpriteRendererProperties}
   {rendererClass=&NiagaraLightRendererProperties, rendererIndex=1}
   {rendererClass=&NiagaraMeshRendererProperties, rendererIndex=2}
   {rendererClass=&NiagaraSpriteRendererProperties, rendererIndex=3}
  ]
 }
 Impact_Streaks {
  emitterUpdateScript [
   &EmitterState()
   &SpawnBurst_Instantaneous(
    Spawn Count=&RandomRangeInt(Minimum=50, Maximum=80)
    Use Spawn Probability=false
    UseLoopCountLimit=false
   )
   &SpawnBurst_Instantaneous(
    Spawn Count=&RandomRangeInt(Minimum=50, Maximum=80)
    Spawn Time=0.08
    Use Spawn Probability=false
    UseLoopCountLimit=false
   ) {module=SpawnBurst_Instantaneous001}
   &SpawnBurst_Instantaneous(
    Spawn Count=&RandomRangeInt(Minimum=50, Maximum=80)
    Spawn Time=0.12
    Use Spawn Probability=false
    UseLoopCountLimit=false
   ) {module=SpawnBurst_Instantaneous002}
  ]
  particleSpawnScript [
   &InitializeParticle(
    Color=&SelectLinearColorFromArray(
     Array Sampling Mode=::NewEnumerator0
     Direct Array Index=?
     Color Selection Array=DataInterface() {propertyValues {ColorData [<rgba 5 2.6961 1.0188 1>]}}
     Random Seed=0
    )
    Lifetime Max=0.25
    Lifetime Min=0.12
    Lifetime Mode=::NewEnumerator1
    Position Offset=<0 0 10>
    Sprite Size Max=<2 80>
    Sprite Size Min=<2 50>
    Sprite Size Mode=::NewEnumerator2
    UsePositionOffset=true
    Write Material Random=true
   )
   &PartitionParticles()
   &SkeletalMeshLocation(
    Bone / Socket Index=$Particles.PartitionParticles.Partition
    MaskSpawnGroup=false
    Mesh Sampling Type=::NewEnumerator5
    Sampled Position Offset=&SelectVectorFromArray(
     Direct Array Index=$Particles.PartitionParticles.Partition
     Vector Selection Array=DataInterface() {propertyValues {FloatData [<0 0 -3>, <0 0 0>]}}
    )
    Skeletal Mesh=DataInterface() {propertyValues {
      DefaultMesh=null
      ExcludeBoneName=root
      OptionalComponentClass=null
      PreviewMesh=&ScoopPluck_Gadget
      SkinningMode=SkinOnTheFly
      SoftSourceActor=null
      SourceMode=Default
      UvSetIndex=0
      WholeMeshLOD=-1
      bExcludeBone=false
      bReadDeformedGeometry=true
      bRequireCurrentFrameData=true
      ComponentTags []
      FilteredBones []
      FilteredSockets [JNT_R_ROCKET_SOCKET, JNT_L_ROCKET_SOCKET]
      MeshUserParameter {parameter {name=None, typeDefHandle {registeredTypeIndex=1}, varData []}}
      SamplingRegions []
     }}
    Socket Sampling Mode=::NewEnumerator5
   )
   &SubUVAnimation(End Frame=3, SubUV Animation Mode:ENiagara_SubUVLookupMode=::NewEnumerator2)
   &ShapeLocation(MaskSpawnGroup=false, Override Seed=false, Sphere Radius=5, UseOwnerScale=true)
   &AddVelocity(
    Cone Angle=50
    Cone Axis=&SelectVectorFromArray(
     Direct Array Index=$Particles.PartitionParticles.Partition
     Vector Selection Array=DataInterface() {propertyValues {FloatData [<0.3 0 1>, <-0.4 0 1>]}}
    )
    Module.UseConeFalloff=false
    Module.UseVelDistribution=false
    Override Seed=false
    Velocity Mode=::NewEnumerator2
    Velocity Speed=&RandomRangeFloat(Minimum=50, Maximum=75)
   )
  ]
  particleUpdateScript [
   &SolveForcesAndVelocity(Clamp Velocity=false, Limit Acceleration=false)
   &SkeletalMeshLocation(
    Bone / Socket Index=$Particles.PartitionParticles.Partition
    MaskSpawnGroup=false
    Mesh Sampling Type=::NewEnumerator5
    Orientation Sampling=::NewEnumerator1
    Sampled Position Offset=&SelectVectorFromArray(
     Direct Array Index=$Particles.PartitionParticles.Partition
     Vector Selection Array=DataInterface() {propertyValues {FloatData [<0 0 -3>, <0 0 0>]}}
    )
    Skeletal Mesh=DataInterface() {propertyValues {
      DefaultMesh=null
      ExcludeBoneName=root
      OptionalComponentClass=null
      PreviewMesh=&ScoopPluck_Gadget
      SkinningMode=SkinOnTheFly
      SoftSourceActor=null
      SourceMode=Default
      UvSetIndex=0
      WholeMeshLOD=-1
      bExcludeBone=false
      bReadDeformedGeometry=true
      bRequireCurrentFrameData=true
      ComponentTags []
      FilteredBones []
      FilteredSockets [JNT_R_ROCKET_SOCKET, JNT_L_ROCKET_SOCKET]
      MeshUserParameter {parameter {name=None, typeDefHandle {registeredTypeIndex=1}, varData []}}
      SamplingRegions []
     }}
    Socket Sampling Mode=::NewEnumerator5
   ) {module=SkeletalMeshLocation001}
   &ParticleState()
   &ScaleColor(
    Scale Alpha=&FloatFromCurve(FloatCurve=DataInterface() {propertyValues {Curve {keys [{arriveTangent=-3.3102, interpMode=RCIM_Cubic, leaveTangent=-3.3102, tangentMode=RCTM_User, time=0, value=1}, <at 1 0 RCIM_Cubic>]}}})
    Scale RGB=&MakeVectorFromLinearColorRGB(Color=&ColorFromCurve(ColorCurve=DataInterface() {propertyValues {
        AlphaCurve {keys []}
        BlueCurve {keys [<at 0 5.481 RCIM_Cubic>,<at 0.125 0.1661>,<at 0.6298 0.0192 RCIM_Cubic>]}
        GreenCurve {keys [<at 0 7.5935 RCIM_Cubic>,<at 0.125 0.5559>,<at 0.6298 0.098 RCIM_Cubic>]}
        RedCurve {keys [<at 0 10 RCIM_Cubic>,<at 0.125 1>,<at 0.6298 0.1878 RCIM_Cubic>]}
       }}))
    ScaleA=true
    ScaleRGB=true
   ) {enabled=false}
   &ScaleColor(
    Scale Alpha=&FloatFromCurve(FloatCurve=DataInterface() {propertyValues {Curve {keys [{arriveTangent=-3.3102, interpMode=RCIM_Cubic, leaveTangent=-3.3102, tangentMode=RCTM_User, time=0, value=1}, <at 1 0 RCIM_Cubic>]}}})
    Scale RGB=&MakeVectorFromLinearColorRGB(Color=&ColorFromCurve(ColorCurve=DataInterface() {propertyValues {
        AlphaCurve {keys []}
        BlueCurve {keys [<at 0 10 RCIM_Cubic>,<at 0.125 1>,<at 0.6298 0.1878 RCIM_Cubic>]}
        GreenCurve {keys [<at 0 10 RCIM_Cubic>,<at 0.125 1>,<at 0.6298 0.1878 RCIM_Cubic>]}
        RedCurve {keys [<at 0 10 RCIM_Cubic>,<at 0.125 1>,<at 0.6298 0.1878 RCIM_Cubic>]}
       }}))
    ScaleA=true
    ScaleRGB=true
   ) {module=ScaleColor001}
   &ScaleSpriteSize(
    Non-Uniform Curve Scale=<1 0.5>
    Non-Uniform Curve Sprite Scale=DataInterface() {propertyValues {
      XCurve {keys [
        {arriveTangent=0.9067, interpMode=RCIM_Cubic, leaveTangent=0.9067, tangentMode=RCTM_User, time=0, value=0.8}
        {arriveTangentWeight=0.2034, interpMode=RCIM_Cubic, leaveTangentWeight=0.2034, tangentMode=RCTM_User, time=0.5, value=1}
        {arriveTangent=-3.1509, interpMode=RCIM_Cubic, leaveTangent=-3.1509, tangentMode=RCTM_User, time=1, value=0.1}
       ]}
      YCurve {keys [
        {arriveTangent=3.7985, interpMode=RCIM_Cubic, leaveTangent=3.7985, tangentMode=RCTM_User, time=0, value=0}
        {arriveTangent=2, interpMode=RCIM_Cubic, leaveTangent=2, time=0.5, value=1.6}
        <at 1 2 RCIM_Cubic>
       ]}
     }}
    Scale Sprite Size Mode=::NewEnumerator3
   )
   &FN_Light_Attributes(
    Color=&ColorFromCurve(ColorCurve=DataInterface() {propertyValues {
       AlphaCurve {keys [<at 0 1 RCIM_Cubic>,<at 1 0 RCIM_Cubic>]}
       BlueCurve {keys [<at 0 0 RCIM_Cubic>]}
       GreenCurve {keys [<at 0 0.5139 RCIM_Cubic>]}
       RedCurve {keys [<at 0 1 RCIM_Cubic>]}
      }})
    Create Color Attribute=true
    Create Exponent Attribute=true
    Create Position Attribute=true
    Create Radius Attribute=true
    Position=<0 -20 150>
    Radius=120
    Scale Color=35
   ) {enabled=false}
   /SetVariables(
    Particles.Linear Color=&MultiplyLinearColor(Linear Color=$Particles.Color, Scale Factor=<1 0.5 0.5 1>)
    Particles.Vector 2D:Vector2f=&Multiply_Vector2D_ByFloat(Float=1.2, Vector2D=$Particles.SpriteSize)
   ) {isSetParameters=true}
  ]
  rendererClasses [&NiagaraSpriteRendererProperties, &NiagaraLightRendererProperties]
  renderers [
   {rendererClass=&NiagaraSpriteRendererProperties}
   {rendererClass=&NiagaraLightRendererProperties, rendererIndex=1}
   {rendererClass=&NiagaraSpriteRendererProperties, rendererIndex=2}
  ]
 }
 RibbonSmoke {
  emitterUpdateScript [
   &EmitterState(
    Enable Distance Culling=true
    MaxDistance=2500000
    Module.bUseMaxDistance=true
    Module.bUseMinDistance=false
    Scalability Mode=::NewEnumerator1
   )
   &SpawnParticlesFromOtherEmitter(
    Attribute Reader=DataInterface() {propertyValues {EmitterBinding {bindingMode=Other, emitterName=FireworkFlare_Parent001}}}
    Spawn Rate=&FloatFromCurve(
     FloatCurve=DataInterface() {propertyValues {Curve {keys [
         <at 0 1>
         {arriveTangent=-0.1867, interpMode=RCIM_Cubic, leaveTangent=-0.1867, tangentMode=RCTM_User, time=0.4, value=1}
         {arriveTangent=-5.8462, interpMode=RCIM_Cubic, leaveTangent=-5.8462, tangentMode=RCTM_User, time=0.6, value=0}
        ]}}}
     CurveIndex=$Emitter.NormalizedLoopAge
     Scale Curve=110
    )
    Spawn Rate Per Particle Cap=125
    Use Maximum Age=false
    Use Minimum Age=false
   )
  ]
  particleSpawnScript [
   &InitializeRibbon(
    Lifetime=&RandomRangeFloat(Minimum=0.8, Maximum=1.5)
    Position Mode:ENiagara_PositionInitializationMode=::NewEnumerator1
    Ribbon Width=1.2
    UsePositionOffset=false
    Write Material Random=true
   ) {module=InitializeRibbon001}
   &SampleParticlesFromOtherEmitter(
    Age Sampling:ENiagara_DirectReadSamplingMode=::NewEnumerator0
    Applied RGB Scale:Vector3f=&Multiply_VectorByFloat(Vector=<1 1 1>, Float=1.2)
    Applied Velocity Scale=0
    Color Sampling=::NewEnumerator0
    Lifetime Application:ENiagara_LifetimeInheritanceOptions=::NewEnumerator1
    Lifetime Sampling:ENiagara_DirectReadSamplingMode=::NewEnumerator0
    Particle ID Sampling=::NewEnumerator1
    Reject Particle If Sampled Index Is Invalid=false
    Sprite Size Sampling:ENiagara_DirectReadSamplingMode=::NewEnumerator0
    Velocity Sampling=::NewEnumerator0
   )
  ]
  particleUpdateScript [
   &ParticleState()
   &ScaleColor(
    Linear Color Curve=DataInterface() {propertyValues {
      AlphaCurve {keys [<at 0 1 RCIM_Cubic>, {arriveTangent=-1, interpMode=RCIM_Cubic, leaveTangent=-1, time=0.0561, value=1}, <at 0.5349 1>, <at 1 0 RCIM_Cubic>]}
      BlueCurve {keys [<at 0.0533 1>,<at 1 0>]}
      GreenCurve {keys [<at 0.0533 1>,<at 1 0>]}
      RedCurve {keys [<at 0.0533 1>,<at 1 0>]}
     }}
    Scale Mode=::NewEnumerator2
   )
   &ScaleRibbonWidth(Ribbon Width Scale=&FloatFromCurve(
     FloatCurve=DataInterface() {propertyValues {Curve {keys [{arriveTangent=-0.9162, interpMode=RCIM_Cubic, leaveTangent=-0.9162, tangentMode=RCTM_User, time=0.1, value=1}, <at 2 0 RCIM_Cubic>]}}}
     Scale Curve=15
    )) {module=ScaleRibbonWidth001}
   /Niagara/Modules/Update/Forces/CurlNoiseForce(
    Noise Frequency=1
    Noise Strength=&Multiply_Float(
     A=250
     B=&FloatFromCurve(
      FloatCurve=DataInterface() {propertyValues {Curve {keys [
          {arriveTangent=1.9891, interpMode=RCIM_Cubic, leaveTangent=1.9891, tangentMode=RCTM_User, time=0, value=0}
          {interpMode=RCIM_Cubic, tangentMode=RCTM_User, time=0.75, value=1}
          {tangentMode=RCTM_User, time=1, value=0}
         ]}}}
      Scale Curve=0.2
     )
    )
    Pan Noise=true
   )
   &SolveForcesAndVelocity(Clamp Velocity=false, Limit Acceleration=false)
   &DynamicMaterialParameters(
    Index 0 Param 1=-1
    Index 0 Param 2=1
    Index 0 Param 4=&FloatFromCurve(FloatCurve=DataInterface() {propertyValues {Curve {keys [
         {arriveTangent=-2.8039, interpMode=RCIM_Cubic, time=0, value=1}
         {arriveTangent=-3.1818, interpMode=RCIM_Cubic, leaveTangent=-3.1818, tangentMode=RCTM_User, time=1, value=0}
        ]}}})
    Module.Index0Float0Write=true
    Module.Index0Float1Write=true
    Module.Index0Float2Write=true
    Module.Index0Float3Write=true
   )
   &FN_Light_Attributes(
    Color=$Particles.Color
    Create Color Attribute=true
    Create Exponent Attribute=true
    Create Radius Attribute=true
    Exponent=4
    Radius=120
    Scale Color=&FloatFromCurve(
     FloatCurve=DataInterface() {propertyValues {Curve {keys [
         {interpMode=RCIM_Cubic, tangentMode=RCTM_User, time=0, value=1}
         {arriveTangent=-2.6287, interpMode=RCIM_Cubic, leaveTangent=-2.6287, tangentMode=RCTM_User, time=1, value=0}
        ]}}}
     Scale Curve=0.25
    )
   ) {enabled=false}
  ]
  rendererClasses [&NiagaraRibbonRendererProperties, &NiagaraLightRendererProperties]
  renderers [
   {rendererClass=&NiagaraRibbonRendererProperties}
   {rendererClass=&NiagaraLightRendererProperties, rendererIndex=1}
  ]
 }
 Spark_High {
  emitterUpdateScript [
   &EmitterState()
   &SpawnPerFrame(
    Spawn:NiagaraBool=$System.DeathSpawn
    Spawn Count:NiagaraInt32=&SelectIntFromArray(
     Direct Array Index=$Engine.QualityLevel
     Direct Array Mode=::NewEnumerator1
     Int Selection Array=DataInterface() {propertyValues {IntData [30,100,120,200,200]}}
    )
    Use Spawn Probability=false
   )
  ]
  particleSpawnScript [
   &InitializeParticle(
    Color=$System.DeathColor
    Lifetime Max=1.6
    Lifetime Min=0.7
    Lifetime Mode=::NewEnumerator1
    Position=$System.DeathPosition
    Position Mode=::NewEnumerator1
    Sprite Size Mode=::NewEnumerator1
    Uniform Sprite Size Max=3
    Uniform Sprite Size Min=1
    UsePositionOffset=false
    Write Material Random=true
   )
   &ShapeLocation(MaskSpawnGroup=false, Override Seed=false, Sphere Radius=22, UseOwnerScale=true)
   &AddVelocity(
    Module.Use Falloff Distance=false
    Override Seed=false
    Velocity Mode=::NewEnumerator1
    Velocity Origin=$System.DeathPosition
    Velocity Speed=&RandomRangeFloat(Minimum=300, Maximum=500)
   )
   &AddVelocityFromPoint(
    Module.Use Falloff Distance=false
    Origin Offset=<75 0 125>
    Velocity Strength=&RandomRangeFloat(Minimum=200, Maximum=325)
   ) {enabled=false}
   /Niagara/Modules/Update/Forces/V2/CurlNoiseForce(Noise Frequency=65, Noise Strength=450, Pan Noise=true) {enabled=false}
   &ApplyInitialForces()
  ]
  particleUpdateScript [
   &ParticleState()
   &GravityForce(Gravity=&RandomRangeVector(Minimum=<0 0 -250>, Maximum=<0 0 -350>))
   &Drag(Drag=&RandomRangeFloat(Minimum=6, Maximum=11), Use Linear Drag=true, Use Rotational Drag=true)
   &SolveForcesAndVelocity(Clamp Velocity=false, Limit Acceleration=false)
   &ScaleSpriteSize(Uniform Curve Sprite Scale=DataInterface() {propertyValues {Curve {keys [{tangentMode=RCTM_User, time=0, value=1}, {tangentMode=RCTM_User, time=1, value=0}]}}})
   &ScaleSpriteSizeBySpeed(Max Scale Factor=<0.2 5>, Min Scale Factor=<1 0.5>)
   &ScaleColor(
    Scale Alpha=&FloatFromCurve(FloatCurve=DataInterface() {propertyValues {Curve {keys [<at 0 1>,<at 0.3 1>,<at 1 0>]}}})
    Scale RGB=&VectorFromFloat(Value=&Multiply_Float(A=$System.TimeOfDayBrightnessMultiplier, B=1.5))
    ScaleA=true
    ScaleRGB=true
   )
   &ScaleColor(
    Scale Alpha=&FloatFromCurve(
     FloatCurve=DataInterface() {propertyValues {Curve {defaultValue=1, keys [{tangentMode=RCTM_User, time=0, value=0}, <at 0.2096 0>, {tangentMode=RCTM_User, time=1, value=1}]}}}
     CurveIndex=&RemapRange(
      Input Value To Remap=/Niagara/DynamicInputs/Camera/CameraFade(
       Camera Query:NiagaraDataInterfaceCamera=DataInterface()
       Fade Offset:NiagaraFloat=0
       Fade Distance:NiagaraFloat=1500
       Use Owner Position:NiagaraBool=true
      )
      Input Min=0.06
      Input Max=0.21
      Output Min=0
      Output Max=1
     )
    )
    ScaleA=true
    ScaleRGB=true
   ) {module=ScaleColor001}
  ]
  rendererClasses [&NiagaraSpriteRendererProperties]
  renderers [{rendererClass=&NiagaraSpriteRendererProperties}]
 }
 TrailEmbers001 {
  emitterUpdateScript [
   &EmitterState()
   &SpawnParticlesFromOtherEmitter(
    Attribute Reader=DataInterface() {propertyValues {EmitterBinding {bindingMode=Other, emitterName=FireworkFlare_Parent001}}}
    Spawn Rate=&FloatFromCurve(
     FloatCurve=DataInterface() {propertyValues {Curve {keys [
         <at 0 1 RCIM_Cubic>
         {arriveTangent=-1.0933, interpMode=RCIM_Cubic, leaveTangent=-1.0933, tangentMode=RCTM_User, time=0, value=1}
         {arriveTangent=-0.9756, interpMode=RCIM_Cubic, leaveTangent=-0.9756, tangentMode=RCTM_User, time=0.2, value=1}
         {arriveTangent=-8.2278, interpMode=RCIM_Cubic, leaveTangent=-8.2278, tangentMode=RCTM_User, time=0.7, value=0.0064}
        ]}}}
     CurveIndex=$Emitter.NormalizedLoopAge
     Scale Curve=75
    )
    Spawn Rate Per Particle Cap=150
    Use Maximum Age=false
    Use Minimum Age=false
   )
  ]
  particleSpawnScript [
   &InitializeParticle(
    Color=<rgba 15 15 15 1>
    Lifetime Max=1.3
    Lifetime Min=0.3
    Lifetime Mode=::NewEnumerator1
    Sprite Rotation Mode=::NewEnumerator2
    Sprite Size Mode=::NewEnumerator1
    Uniform Sprite Size Max=4
    Uniform Sprite Size Min=2
    UsePositionOffset=false
    Write Material Random=true
   )
   &SampleParticlesFromOtherEmitter(Apply Sampled Alpha=false, Color Sampling=::NewEnumerator0, Transform Data Option=1)
   &SphereLocation(
    MaskSpawnGroup=false
    Override Seed=false
    Sphere Radius=10
    Surface Only=false
    UseNonUniformScale=false
    UseOffset=false
    UseSphereOrientation=false
   )
  ]
  particleUpdateScript [
   &ParticleState()
   &CurlNoiseForce(
    Noise Frequency=&RandomRangeFloat(Minimum=0.25, Maximum=1)
    Noise Strength=&RandomRangeFloat(Minimum=120, Maximum=300)
    Pan Noise=true
    Pan Noise Field=&RandomRangeVector(Minimum=<0 0 0>, Maximum=<1 1 1>)
   )
   &GravityForce(Gravity=&RandomRangeVector(Minimum=<0 0 -200>, Maximum=<0 0 -250>))
   &Drag(Drag=&RandomRangeFloat(Minimum=3, Maximum=8), Use Linear Drag=true, Use Rotational Drag=false)
   &SolveForcesAndVelocity(Clamp Velocity=false, Limit Acceleration=false)
   &ScaleSpriteSizeBySpeed(Max Scale Factor=<0.3 3>, Velocity Threshold=100)
   &ScaleSpriteSize(
    Uniform Curve Scale=0.5
    Uniform Curve Sprite Scale=DataInterface() {propertyValues {Curve {keys [{interpMode=RCIM_Cubic, tangentMode=RCTM_User, time=0, value=1}, {interpMode=RCIM_Cubic, tangentMode=RCTM_User, time=1, value=0}]}}}
   )
   &ScaleColor(
    Scale Alpha=&FloatFromCurve(FloatCurve=DataInterface() {propertyValues {Curve {keys [{arriveTangent=-1.933, interpMode=RCIM_Cubic, leaveTangent=-1.933, tangentMode=RCTM_User, time=0, value=1}, <at 1 0>]}}})
    Scale RGB=&VectorFromFloat(Value=&FloatFromCurve(
      FloatCurve=DataInterface() {propertyValues {Curve {keys [{arriveTangent=-2.5197, interpMode=RCIM_Cubic, leaveTangent=-2.5197, tangentMode=RCTM_User, time=0, value=1}, <at 1 0>]}}}
      Scale Curve=&RandomRangeFloat(Minimum=8, Maximum=15)
     ))
    ScaleA=false
    ScaleRGB=true
   )
   &ScaleColor(
    Scale Alpha=&FloatFromCurve(
     FloatCurve=DataInterface() {propertyValues {Curve {defaultValue=1, keys [{tangentMode=RCTM_User, time=0, value=0}, <at 0.2096 0>, {tangentMode=RCTM_User, time=1, value=1}]}}}
     CurveIndex=&RemapRange(
      Input Value To Remap=&CameraFade(
       Camera Query:NiagaraDataInterfaceCamera=DataInterface()
       Fade Offset:NiagaraFloat=0
       Fade Distance:NiagaraFloat=1500
       Use Owner Position:NiagaraBool=true
      )
      Input Min=0.06
      Input Max=0.21
      Output Min=0
      Output Max=1
     )
    )
    ScaleA=true
    ScaleRGB=true
   ) {module=ScaleColor001}
  ]
  rendererClasses [&NiagaraSpriteRendererProperties]
  renderers [{rendererClass=&NiagaraSpriteRendererProperties}]
 }
}
summary {
 systemName=NS_Emote_ScoopPluck_Firework1
 emitters [
  {emitterName=TrailEmbers001, rendererClasses [&NiagaraSpriteRendererProperties]}
  {emitterName=RibbonSmoke, rendererClasses [&NiagaraRibbonRendererProperties, &NiagaraLightRendererProperties]}
  {
   emitterName=FireworkFlare_Parent001
   rendererClasses [&NiagaraSpriteRendererProperties, &NiagaraLightRendererProperties]
  }
  {
   emitterName=Impact_Streaks
   rendererClasses [&NiagaraSpriteRendererProperties, &NiagaraLightRendererProperties]
  }
  {
   emitterName=Glow
   rendererClasses [
    &NiagaraSpriteRendererProperties
    &NiagaraLightRendererProperties
    &NiagaraMeshRendererProperties
   ]
  }
  {
   emitterName=Flare_With_Streak002
   rendererClasses [&NiagaraSpriteRendererProperties, &NiagaraLightRendererProperties]
  }
  {emitterName=Spark_High, rendererClasses [&NiagaraSpriteRendererProperties]}
 ]
 userVariables [
  User.BaseCurveColor {
   type=NiagaraDataInterfaceColorCurve
   description=""
   defaultValue {
    struct=NiagaraExt_VariableValue_DataInterface
    value {
     dataInterface=&self:NiagaraDataInterfaceColorCurve_0
     dataInterfaceClass=&NiagaraDataInterfaceColorCurve
    }
   }
  }
  User.BaseCurveColorHueShift {
   type=NiagaraDataInterfaceCurve
   description=""
   defaultValue {
    struct=NiagaraExt_VariableValue_DataInterface
    value {
     dataInterface=&self:NiagaraDataInterfaceCurve_0
     dataInterfaceClass=&NiagaraDataInterfaceCurve
    }
   }
  }
  User.BaseCurveColorPick {type=NiagaraFloat, defaultValue=0, description=""}
  User.bIsFrontEnd {type=NiagaraBool}
  User.bIsFrontEndPreview {type=NiagaraBool}
 ]
}
