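# Niagara effect for the ScoopPluck emote that draws short-lived thruster flame meshes on the gadget prop.
# Each burst spawns two particles, partitioned so each follows the left or right socket on the skeletal mesh.
# Curves stretch and widen the flames, scaled by User.Flame_Size, while color shifts from orange toward pale and alpha fades.
# Two mesh renderers and a second color and scale attribute set provide a layered flame.
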NiagaraSystem /BRCosmetics/Animation/Game/MainPlayer/Emotes/ScoopPluck/FX/NS_Emote_ScoopPluck_Thruster1
dependencies {
 usedDataInterfaces [&NiagaraDataInterfaceSkeletalMesh, &NiagaraDataInterfaceColorCurve, &NiagaraDataInterfaceCurve]
 usedDynamicInputs []
 usedModules [&SystemState]
}
emitters {Comet_Flame_Mesh {
  emitterUpdateScript [
   &EmitterState(Life Cycle Mode=::NewEnumerator1, Loop Delay=0.2, Loop Duration=0.1, UseLoopDelay=false)
   &SpawnBurst_Instantaneous(Spawn Count=2, Use Spawn Probability=false)
  ]
  particleSpawnScript [
   &InitializeParticle(
    Color=<rgba 1.5 0.375 0.15 5>
    Lifetime=0.2
    Mass Mode=::NewEnumerator1
    Mesh Scale=<-0.04 0.035 0.035>
    Mesh Scale Mode=::NewEnumerator4
    UsePositionOffset=false
    Write Material Random=true
   ) {module=InitializeParticle001}
   &PartitionParticles()
   &InitialMeshOrientation(Mesh Orientation Mode=::NewEnumerator2, Orientation Vector=<-1 0 0>, Use Rotation Vector=false)
   /SetVariables(Particles.Color_2:LinearColor=<rgba 20 12.5 2 0.25>, Particles.Scale_2:Vector3f=<0.035 0.025 0.025>) {isSetParameters=true}
   &DynamicMaterialParameters(
    Index 0 Param 1=5
    Index 0 Param 2=5
    Module.Index0Float0Write=true
    Module.Index0Float1Write=true
    Module.Index0Float2Write=true
    Module.Index0Float3Write=true
   )
  ]
  particleUpdateScript [
   &SkeletalMeshLocation(
    Bone / Socket Index=$Particles.PartitionParticles.Partition
    MaskSpawnGroup=false
    Mesh Sampling Type=::NewEnumerator5
    Skeletal Mesh=DataInterface() {propertyValues {
      DefaultMesh=null
      ExcludeBoneName=root
      OptionalComponentClass=null
      PreviewMesh=/BRCosmetics/Animation/Props/Gadgets_ScoopPluck/Meshes/ScoopPluck_Gadget
      SkinningMode=SkinOnTheFly
      SoftSourceActor=null
      SourceMode=Default
      UvSetIndex=0
      WholeMeshLOD=-1
      bExcludeBone=false
      bReadDeformedGeometry=true
      bRequireCurrentFrameData=true
      ComponentTags []
      FilteredBones []
      FilteredSockets [JNT_L_SOCKET,JNT_R_SOCKET]
      MeshUserParameter {parameter {name=None, typeDefHandle {registeredTypeIndex=1}, varData []}}
      SamplingRegions []
     }}
    Socket Sampling Mode=::NewEnumerator5
   )
   &ParticleState()
   &ScaleColor(
    Scale Alpha=&Lerp_Float(A=1, B=0, Alpha=$Particles.NormalizedAge)
    Scale RGB=&MakeVectorFromLinearColorRGB(Color=&ColorFromCurve(ColorCurve=DataInterface() {propertyValues {
        AlphaCurve {keys []}
        BlueCurve {keys [<at 0 0.25 RCIM_Cubic>,<at 0.492 0.5>,<at 1 0.875 RCIM_Cubic>]}
        GreenCurve {keys [<at 0 0.4375 RCIM_Cubic>,<at 0.492 0.625>,<at 1 0.9062 RCIM_Cubic>]}
        RedCurve {keys [<at 0 1 RCIM_Cubic>,<at 0.492 1>,<at 1 1 RCIM_Cubic>]}
       }}))
    ScaleA=true
    ScaleRGB=true
   )
   &ScaleMeshSize(Scale Factor=&MakeVector(
     X=&Multiply_Float(
      A=$User.Flame_Size
      B=&FloatFromCurve(FloatCurve=DataInterface() {propertyValues {Curve {keys [<at 0 0.75 RCIM_Cubic>, {arriveTangent=-0.25, interpMode=RCIM_Cubic, leaveTangent=-0.25, time=0.5, value=2.5}, <at 1 0.5 RCIM_Cubic>]}}})
     )
     Y=&Multiply_Float(
      A=$User.Flame_Size
      B=&FloatFromCurve(FloatCurve=DataInterface() {propertyValues {Curve {keys [<at 0.5 1 RCIM_Cubic>,<at 1 1.75 RCIM_Cubic>]}}})
     )
     Z=&Multiply_Float(
      A=$User.Flame_Size
      B=&FloatFromCurve(FloatCurve=DataInterface() {propertyValues {Curve {keys [<at 0.5 1 RCIM_Cubic>,<at 1 1.75 RCIM_Cubic>]}}})
     )
    ))
   /SetVariables(Particles.Scale_2:Vector3f=&Multiply_Vector(
     A=$Particles.Initial.Scale_2
     B=&MakeVector(
      X=&FloatFromCurve(
       FloatCurve=DataInterface() {propertyValues {Curve {keys [<at 0 0.75 RCIM_Cubic>, {arriveTangent=-0.25, interpMode=RCIM_Cubic, leaveTangent=-0.25, time=0.5, value=2.5}, <at 1 0.5 RCIM_Cubic>]}}}
       Scale Curve=$User.Flame_Size
      )
      Y=&FloatFromCurve(FloatCurve=DataInterface() {propertyValues {Curve {keys [<at 0.5 1 RCIM_Cubic>,<at 1 1.75 RCIM_Cubic>]}}}, Scale Curve=$User.Flame_Size)
      Z=&FloatFromCurve(FloatCurve=DataInterface() {propertyValues {Curve {keys [<at 0.5 1 RCIM_Cubic>,<at 1 1.75 RCIM_Cubic>]}}}, Scale Curve=$User.Flame_Size)
     )
    )) {isSetParameters=true}
  ]
  rendererClasses [&NiagaraMeshRendererProperties]
  renderers [
   {rendererClass=&NiagaraMeshRendererProperties, rendererIndex=0}
   {rendererClass=&NiagaraMeshRendererProperties, rendererIndex=1}
  ]
 }}
summary {
 systemName=NS_Emote_ScoopPluck_Thruster1
 emitters [{emitterName=Comet_Flame_Mesh, rendererClasses [&NiagaraMeshRendererProperties]}]
 userVariables [User.Flame_Size {type=NiagaraFloat, defaultValue=1, description=""}]
}
