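# Looping Niagara thruster effect for the ScoopPluck emote gadget, attached to its left and right rocket sockets by per-particle partition.
# Comet_Flame_Mesh bursts short-lived meshes through two renderers, with a second scale and color set, animating stretch, tint and fade over life.
# Rim_Glow adds camera-offset sprites and particle lights that outlive their lifetime.
# User.Flame_Size scales flame and glow sprites, and the flipbook Smoke emitter is disabled.
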NiagaraSystem /BRCosmetics/Animation/Game/MainPlayer/Emotes/ScoopPluck/FX/NS_ScoopPluck_ThrusterFX_Looping_03
dependencies {
 usedDataInterfaces [&NiagaraDataInterfaceSocketReader, &NiagaraDataInterfaceColorCurve, &NiagaraDataInterfaceCurve]
 usedDynamicInputs []
 usedModules [&SystemState]
}
emitters {
 Comet_Flame_Mesh {
  emitterUpdateScript [
   &EmitterState(Life Cycle Mode=::NewEnumerator1, Loop Delay=0.2, Loop Duration=0.1, UseLoopDelay=false)
   &SpawnBurst_Instantaneous(Spawn Count=2, Use Spawn Probability=false)
  ]
  particleSpawnScript [
   &InitializeParticle(
    Color=<rgba 1.5 0.375 0.15 5>
    Lifetime=0.2
    Mass Mode=::NewEnumerator1
    Mesh Scale=<-0.04 0.035 0.035>
    Mesh Scale Mode=::NewEnumerator4
    UsePositionOffset=false
    Write Material Random=true
   ) {module=InitializeParticle001}
   &InitialMeshOrientation(Mesh Orientation Mode=::NewEnumerator2, Orientation Vector=<-1 0 0>, Use Rotation Vector=false)
   &PartitionParticles()
   /SetVariables(Particles.Color_2:LinearColor=<rgba 20 12.5 2 0.25>, Particles.Scale_2:Vector3f=<0.035 0.025 0.025>) {isSetParameters=true}
   &DynamicMaterialParameters(
    Index 0 Param 1=5
    Index 0 Param 2=5
    Module.Index0Float0Write=true
    Module.Index0Float1Write=true
    Module.Index0Float2Write=true
    Module.Index0Float3Write=true
   )
  ]
  particleUpdateScript [
   &SocketLocation(
    SamplePosOffset=true
    Sampled Position Offset=<0 0 -25>
    Sampled Position Offset Coordinate Space:ENiagaraSkelMeshTransforms=::NewEnumerator2
    Socket Index=$Particles.PartitionParticles.Partition
    Socket Reader=DataInterface() {propertyValues {
      AttachComponentClass=null
      AttachComponentTag=None
      EditorPreviewAsset=/BRCosmetics/Animation/Props/Gadgets_ScoopPluck/Meshes/ScoopPluck_Gadget
      SourceActorRef=null
      SourceAsset=null
      SourceMode=Default
      bRequireCurrentFrameData=true
      bUpdateSocketsPerFrame=true
      FilteredSockets [JNT_L_Rocket,JNT_R_Rocket]
      ObjectParameterBinding {parameter {name=None, typeDefHandle {registeredTypeIndex=1}, varData []}}
     }}
    Socket Sampling Mode=::NewEnumerator3
    VelocitySpeedLimit=false
   )
   &ParticleState()
   &ScaleColor(
    Scale Alpha=&Lerp_Float(A=1, B=0, Alpha=$Particles.NormalizedAge)
    Scale RGB=&MakeVectorFromLinearColorRGB(Color=&ColorFromCurve(ColorCurve=DataInterface() {propertyValues {
        AlphaCurve {keys []}
        BlueCurve {keys [<at 0 0.25 RCIM_Cubic>,<at 0.492 0.5>,<at 1 0.875 RCIM_Cubic>]}
        GreenCurve {keys [<at 0 0.4375 RCIM_Cubic>,<at 0.492 0.625>,<at 1 0.9062 RCIM_Cubic>]}
        RedCurve {keys [<at 0 1 RCIM_Cubic>,<at 0.492 1>,<at 1 1 RCIM_Cubic>]}
       }}))
    ScaleA=true
    ScaleRGB=true
   )
   &ScaleMeshSize(Scale Factor=&MakeVector(
     X=&Multiply_Float(
      A=$User.Flame_Size
      B=&FloatFromCurve(FloatCurve=DataInterface() {propertyValues {Curve {keys [<at 0 0.75 RCIM_Cubic>, {arriveTangent=-0.25, interpMode=RCIM_Cubic, leaveTangent=-0.25, time=0.5, value=2.5}, <at 1 0.5 RCIM_Cubic>]}}})
     )
     Y=&Multiply_Float(
      A=$User.Flame_Size
      B=&FloatFromCurve(FloatCurve=DataInterface() {propertyValues {Curve {keys [<at 0.5 1 RCIM_Cubic>,<at 1 1.75 RCIM_Cubic>]}}})
     )
     Z=&Multiply_Float(
      A=$User.Flame_Size
      B=&FloatFromCurve(FloatCurve=DataInterface() {propertyValues {Curve {keys [<at 0.5 1 RCIM_Cubic>,<at 1 1.75 RCIM_Cubic>]}}})
     )
    ))
   /SetVariables(Particles.Scale_2:Vector3f=&Multiply_Vector(
     A=$Particles.Initial.Scale_2
     B=&MakeVector(
      X=&FloatFromCurve(
       FloatCurve=DataInterface() {propertyValues {Curve {keys [<at 0 0.75 RCIM_Cubic>, {arriveTangent=-0.25, interpMode=RCIM_Cubic, leaveTangent=-0.25, time=0.5, value=2.5}, <at 1 0.5 RCIM_Cubic>]}}}
       Scale Curve=$User.Flame_Size
      )
      Y=&FloatFromCurve(FloatCurve=DataInterface() {propertyValues {Curve {keys [<at 0.5 1 RCIM_Cubic>,<at 1 1.75 RCIM_Cubic>]}}}, Scale Curve=$User.Flame_Size)
      Z=&FloatFromCurve(FloatCurve=DataInterface() {propertyValues {Curve {keys [<at 0.5 1 RCIM_Cubic>,<at 1 1.75 RCIM_Cubic>]}}}, Scale Curve=$User.Flame_Size)
     )
    )) {isSetParameters=true}
  ]
  rendererClasses [&NiagaraMeshRendererProperties]
  renderers [
   {rendererClass=&NiagaraMeshRendererProperties, rendererIndex=0}
   {rendererClass=&NiagaraMeshRendererProperties, rendererIndex=1}
  ]
 }
 Rim_Glow {
  emitterUpdateScript [
   &EmitterState(Life Cycle Mode=::NewEnumerator1, Loop Delay=0.2, Loop Duration=1, UseLoopDelay=false)
   &SpawnBurst_Instantaneous(Spawn Count=8, Spawn Time=0.2, Use Spawn Probability=false)
  ]
  particleSpawnScript [
   &InitializeParticle(
    Color=<rgba 20 6.5 2 0.5>
    Mass Mode=::NewEnumerator1
    Position Offset=<8 0 0>
    Sprite Size=&Multiply_Vector2D_ByFloat(Float=$User.Flame_Size, Vector2D=<4 4>)
    Sprite Size Mode=::NewEnumerator4
    UsePositionOffset=true
    Write Material Random=true
   ) {module=InitializeParticle001}
   &PartitionParticles()
   &CameraOffset(Camera Offset Amount=3)
  ]
  particleUpdateScript [
   &ParticleState(Kill Particles When Lifetime Has Elapsed=false, Let Infinitely Lived Particles Die When Emitter Deactivates=true)
   &SocketLocation(
    SamplePosOffset=true
    Sampled Position Offset=<0 0 -25>
    Sampled Position Offset Coordinate Space:ENiagaraSkelMeshTransforms=::NewEnumerator2
    Socket Index=$Particles.PartitionParticles.Partition
    Socket Reader=DataInterface() {propertyValues {
      AttachComponentClass=null
      AttachComponentTag=None
      EditorPreviewAsset=&ScoopPluck_Gadget
      SourceActorRef=null
      SourceAsset=null
      SourceMode=Default
      bRequireCurrentFrameData=true
      bUpdateSocketsPerFrame=true
      FilteredSockets [JNT_L_Rocket,JNT_R_Rocket]
      ObjectParameterBinding {parameter {name=None, typeDefHandle {registeredTypeIndex=1}, varData []}}
     }}
    Socket Sampling Mode=::NewEnumerator3
    VelocitySpeedLimit=false
   )
   &FN_Light_v3(
    Color=<rgba 1 0.325 0.1 1>
    Create Color Attribute=true
    Create Exponent Attribute=true
    Create Radius Attribute=true
    Enable Particle Light Enable=true
    Exponent=1
    Light Enabled=true
    Radius=10
    Scale Color=40
   )
  ]
  rendererClasses [&NiagaraSpriteRendererProperties, &NiagaraLightRendererProperties]
  renderers [
   {rendererClass=&NiagaraSpriteRendererProperties}
   {rendererClass=&NiagaraLightRendererProperties, rendererIndex=1}
  ]
 }
 Smoke {
  bEnabled=false
  emitterUpdateScript [
   &EmitterState(Life Cycle Mode=::NewEnumerator1, Loop Delay=0.2, Loop Duration=0.1, UseLoopDelay=false)
   &SpawnRate(SpawnRate=4, Use Spawn Probability=false)
  ]
  particleSpawnScript [
   &InitializeParticle(
    Lifetime=&RandomRangeFloat(Minimum=0.4, Maximum=0.6)
    Position Offset=<2 0 0>
    Sprite Rotation Mode=::NewEnumerator2
    Sprite Size Mode=::NewEnumerator3
    Sprite UV Mode=::NewEnumerator3
    Uniform Sprite Size=&RandomRangeFloat(Minimum=40, Maximum=60)
    UsePositionOffset=true
    Write Material Random=true
   ) {module=InitializeParticle002}
   &DynamicMaterialParameters(
    Index 0 Param 2=1
    Index 0 Param 4=1
    Module.Index0Float0Write=true
    Module.Index0Float1Write=true
    Module.Index0Float2Write=true
    Module.Index0Float3Write=true
   )
  ]
  particleUpdateScript [
   &ParticleState()
   &AddVelocity(Velocity=&UniformRangedVector(Minimum=<5 0 0>, Maximum=<15 0 0>))
   &Drag(Use Linear Drag=true, Use Rotational Drag=false)
   &SolveForcesAndVelocity(Clamp Velocity=false, Limit Acceleration=false)
   &DynamicMaterialParameters(
    Index 0 Param 1=&Lerp_Float(A=1, B=-1, Alpha=$Particles.NormalizedAge)
    Index 0 Param 2=1
    Index 0 Param 4=1
    Module.Index0Float0Write=false
    Module.Index0Float1Write=true
    Module.Index0Float2Write=false
    Module.Index0Float3Write=true
   ) {module=DynamicMaterialParameters001}
   /SetVariables(Particles.SubImageIndex=&FloatFromCurve(FloatCurve=DataInterface() {propertyValues {Curve {keys [{arriveTangent=105.1181, interpMode=RCIM_Cubic, leaveTangent=105.1181, tangentMode=RCTM_User, time=0, value=0}, <at 1 63 RCIM_Cubic>]}}})) {isSetParameters=true}
   &ScaleColor(
    Scale Alpha=&FloatFromCurve(FloatCurve=DataInterface() {propertyValues {Curve {keys [
         <at 0 0 RCIM_Cubic>
         {arriveTangent=-0.0259, interpMode=RCIM_Cubic, leaveTangent=-0.0259, tangentMode=RCTM_User, time=0.1, value=1}
         {arriveTangent=0.0671, interpMode=RCIM_Cubic, leaveTangent=0.0671, tangentMode=RCTM_User, time=0.5, value=1}
         <at 1 0 RCIM_Cubic>
        ]}}})
    ScaleA=true
    ScaleRGB=true
   )
   &ScaleColor(
    Scale RGB=&MakeVectorFromLinearColorRGB(Color=&ColorFromCurve(ColorCurve=DataInterface() {propertyValues {
        AlphaCurve {keys []}
        BlueCurve {keys [<at 0 0.25 RCIM_Cubic>,<at 0.6182 0.125 RCIM_Cubic>]}
        GreenCurve {keys [<at 0 0.4375 RCIM_Cubic>,<at 0.6182 0.125 RCIM_Cubic>]}
        RedCurve {keys [<at 0 1 RCIM_Cubic>,<at 0.6182 0.125 RCIM_Cubic>]}
       }}))
    ScaleA=true
    ScaleRGB=true
   ) {module=ScaleColor001}
   &ScaleSpriteSize(Scale Factor=&Vector2DFromFloat(Value=&FloatFromCurve(FloatCurve=DataInterface() {propertyValues {Curve {keys [{arriveTangent=6.2193, interpMode=RCIM_Cubic, leaveTangent=6.2193, tangentMode=RCTM_User, time=0, value=0.2}, <at 1 2 RCIM_Cubic>]}}})))
  ]
  rendererClasses [&NiagaraSpriteRendererProperties]
  renderers [{rendererClass=&NiagaraSpriteRendererProperties}]
 }
}
summary {
 systemName=NS_ScoopPluck_ThrusterFX_Looping_03
 emitters [
  {emitterName=Comet_Flame_Mesh, rendererClasses [&NiagaraMeshRendererProperties]}
  {bEnabled=false, emitterName=Smoke, rendererClasses [&NiagaraSpriteRendererProperties]}
  {emitterName=Rim_Glow, rendererClasses [&NiagaraSpriteRendererProperties, &NiagaraLightRendererProperties]}
 ]
 userVariables [User.Flame_Size {type=NiagaraFloat, defaultValue=1, description=""}]
}
